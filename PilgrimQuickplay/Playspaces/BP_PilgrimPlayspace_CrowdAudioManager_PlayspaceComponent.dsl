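; Playspace component that drives festival crowd audio across Pilgrim Quickplay states.
; On BeginPlay it awaits the event router and listens for state events, dispatching Began to common and MainStage or BattleStage handlers; Ended is unwired.
; Handlers play one-shot crowd stingers and manage a looked-up bed: spawned at Intro, faded out at Pregame and SongResults, faded back in at SongGameplay.
; EndPlay stops listening and destroys the bed.
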(event EventBeginPlay
  (bind _aspilgrim_quickplay_playspace (Utilities|Casting|CastToPilgrimQuickplayPlayspace (Components|GetOwner)))
  (bind _eventrouter (Pilgrim|PilgrimWaitForEventRouter _aspilgrim_quickplay_playspace)
    (:OnEventRouterAvailable
      (bind _output_get (Variables|Default|SetEventRouter _eventrouter))
      (Utilities|IsValid _output_get
        (:Is%20Valid
          (bind (_asynctaskproxy _payload) (GameplayEventRouter|StartListeningToEvent (Variables|Default|GetEventRouter) "/Script/PilgrimQPSharedRuntime.PilgrimEvent_QuickplayState_Began")
            (:then
              (Utilities|Array|Add (Variables|Components|GetEventHandles) _asynctaskproxy)
              (bind (_asynctaskproxy_2 _payload_2) (GameplayEventRouter|StartListeningToEvent (Variables|Default|GetEventRouter) "/Script/PilgrimQPSharedRuntime.PilgrimEvent_QuickplayState_Ended")
                (:then
                  (Utilities|Array|Add (Variables|Components|GetEventHandles) _asynctaskproxy_2))
                (:OnEventReceived
                  ;[1 1 1] Hook this up if need to listen to the end of a particular state
                  (switch Utilities|FlowControl|Switch|SwitchonEPilgrimQuickplayState (Utilities|Struct|BreakPilgrimEventQuickplayStateEnded _payload_2)))))
            (:OnEventReceived
              (switch Utilities|FlowControl|Switch|SwitchonEPilgrimQuickplayState (Utilities|Struct|BreakPilgrimEventQuickplayStateBegan _payload)
                (:Loading
                  (CallFunction|OnLoadingStart)
                  (if (Variables|Default|GetIsListeningToGeneralQuickplayStatesFromMainStage)
                    (CallFunction|OnLoadingStartMainStage)
                    (else
                      (CallFunction|OnLoadingStartBattleStage))))
                (:Pregame
                  (CallFunction|OnPregameStart)
                  (if (Variables|Default|GetIsListeningToGeneralQuickplayStatesFromMainStage)
                    (CallFunction|OnPregameStartMainStage)
                    (else
                      (CallFunction|OnPregameStartBattleStage))))
                (:Preintro
                  (CallFunction|OnPreintroStart)
                  (if (Variables|Default|GetIsListeningToGeneralQuickplayStatesFromMainStage)
                    (CallFunction|OnPreintoStartMainStage)
                    (else
                      (CallFunction|OnPreintroStartBattleStage))))
                (:Intro
                  (CallFunction|OnIntroStart)
                  (if (Variables|Default|GetIsListeningToGeneralQuickplayStatesFromMainStage)
                    (CallFunction|OnIntroStartMainStage)
                    (else
                      (CallFunction|OnIntroStartBattleStage))))
                (:SongGameplay
                  (CallFunction|OnSongGameplayStart)
                  (if (Variables|Default|GetIsListeningToGeneralQuickplayStatesFromMainStage)
                    (CallFunction|OnSongGameplayStartMainStage)
                    (else
                      (CallFunction|OnSongGameplayStartBattleStage))))
                (:Outro
                  (CallFunction|OnOutroStart)
                  (if (Variables|Default|GetIsListeningToGeneralQuickplayStatesFromMainStage)
                    (CallFunction|OnOutroStartMainStage)
                    (else
                      (CallFunction|OnOutroStartBattleStage))))
                (:SongResults
                  (CallFunction|OnSongResultsStart)
                  (if (Variables|Default|GetIsListeningToGeneralQuickplayStatesFromMainStage)
                    (CallFunction|OnSongResultsStartMainStage)
                    (else
                      (CallFunction|OnSongResultsStartBattleStage))))
                (:SetResults
                  (CallFunction|OnSetResultsStart)
                  (if (Variables|Default|GetIsListeningToGeneralQuickplayStatesFromMainStage)
                    (CallFunction|OnSetResultsStartMainStage)
                    (else
                      (CallFunction|OnSetResultsStartBattleStage))))))))))))

(event EventEndPlay
  (params (EndPlayReason :type EEndPlayReason%20Enum :default "Destroyed"))
  (bind _crowdbedaudiohandle (Variables|Components|GetCrowdBedAudioHandle))
  (for _array_element (Variables|Components|GetEventHandles)
    (Utilities|IsValid _array_element
      (:Is%20Valid
        (GameplayEventRouter|StopListeningtoEvent _array_element)
        (Utilities|IsValid _crowdbedaudiohandle
          (:Is%20Valid
            (Audio|Components|Audio|Stop _crowdbedaudiohandle)
            (Components|DestroyComponent _crowdbedaudiohandle)
            (Variables|Components|SetCrowdBedAudioHandle 0)))))))

(event Custom|OnLoadingStart)

(event Custom|OnLoadingStart_MainStage)

(event Custom|OnLoadingStart_BattleStage)

(event Custom|OnPregameStart_MainStage)

(event Custom|OnPregameStart_BattleStage)

(event Custom|OnPregameStart
  (bind _crowdbedaudiohandle (Variables|Components|GetCrowdBedAudioHandle))
  (Utilities|IsValid _crowdbedaudiohandle
    (:Is%20Valid
      (Audio|Components|Audio|FadeOut _crowdbedaudiohandle 2.0 0.0 "Linear" :!class /Script/Engine.AudioComponent))))

(event Custom|OnPreintroStart_BattleStage)

(event Custom|OnPreintoStart_MainStage
  (Audio|PlaySound2D "/SparksMaps/Audio/Crowd/MSS_Festival_Crowd_PreIntro.MSS_Festival_Crowd_PreIntro"))

(event Custom|OnIntroStart_MainStage
  (Audio|PlaySound2D "/SparksMaps/Audio/Crowd/MSS_Festival_Crowd_Intro.MSS_Festival_Crowd_Intro"))

(event Custom|OnIntroStart_BattleStage
  (Audio|PlaySound2D "/SparksMaps/Audio/Crowd/MSS_Festival_Crowd_Intro.MSS_Festival_Crowd_Intro"))

(event Custom|OnPreintroStart)

(event Custom|OnSongGameplayStart_MainStage)

(event Custom|OnSongGameplayStart_BattleStage)

(event Custom|OnOutroStart_MainStage
  (Audio|PlaySound2D "/SparksMaps/Audio/Crowd/Festival_Crowd_Arena_SongEnd.Festival_Crowd_Arena_SongEnd"))

(event Custom|OnOutroStart_BattleStage
  (Audio|PlaySound2D "/SparksMaps/Audio/Crowd/Festival_Crowd_Arena_SongEnd.Festival_Crowd_Arena_SongEnd"))

(event Custom|OnIntroStart
  (bind _crowdbedaudiohandle (Variables|Components|GetCrowdBedAudioHandle))
  ;[1 1 1] Note: Not final. Should experiment with different entry/exit points. Also need to add fade in/out.
  (Utilities|IsValid _crowdbedaudiohandle
    (:Is%20Valid
      (if (Audio|Components|Audio|IsPlaying _crowdbedaudiohandle)
        _
        (else
          _)))
    (:Is%20Not%20Valid
      (bind _returnvalue (Audio|SpawnSound2D "/SparksMaps/Audio/Crowd/MSS_Festival_Crowd_Arena_Gameplay_Bed.MSS_Festival_Crowd_Arena_Gameplay_Bed" 1.0 1.0 0.0 0 false false))
      (Variables|Components|SetCrowdBedAudioHandle _returnvalue)))
  (Audio|Components|Audio|FadeIn _crowdbedaudiohandle 2.0 1.0 0.0 "Linear" :!class /Script/Engine.AudioComponent))

(event Custom|OnSongResultsStart_MainStage)

(event Custom|OnSongResultsStart_BattleStage)

(event Custom|OnSetResultsStart_MainStage)

(event Custom|OnSetResultsStart_BattleStage)

(event Custom|OnSongGameplayStart
  (bind _crowdbedaudiohandle (Variables|Components|GetCrowdBedAudioHandle))
  ;[1 1 1] Note: Not final. Should experiment with different entry/exit points. Also need to add fade in/out.
  (Utilities|IsValid _crowdbedaudiohandle
    (:Is%20Valid
      (if (not (Audio|Components|Audio|IsPlaying _crowdbedaudiohandle))
        ;;[1 1 1] This for re-starting Crowd in Sets and BattleStage, for successive rounds
        (Audio|Components|Audio|FadeIn _crowdbedaudiohandle 2.0 1.0 0.0 "Linear" :!class /Script/Engine.AudioComponent)))))

(event Custom|OnOutroStart)

(event Custom|OnSongResultsStart
  (bind _crowdbedaudiohandle (Variables|Components|GetCrowdBedAudioHandle))
  ;[1 1 1] Note: Not final. Should experiment with different entry/exit points. Also need to add fade in/out.
  (Utilities|IsValid _crowdbedaudiohandle
    (:Is%20Valid
      (Audio|Components|Audio|FadeOut _crowdbedaudiohandle 2.0 0.0 "Linear" :!class /Script/Engine.AudioComponent))))

(event Custom|OnSetResultsStart)
