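; Client-side director component on the Pilgrim Quickplay playspace that runs the band intro cinematic.
; On the StartIntro event it adds intro UI, hides HUD and requests async spotlight montages.
; A state machine drives camera-director shots: group intro with tag-bound pawns, spotlights for local-team players, then cleanup and end-event broadcasts.
; Quick Intro swaps the group shot, advances by timer and can skip spotlights; helpers handle instrument montages and spawn transforms.
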(event EventBeginPlay
  ;[0.8449 1 0] Get associated playspace and event router
  (bind _asbp_pilgrim_quickplay_playspace (Utilities|Casting|CastToBP_PilgrimQuickplayPlayspace (Components|GetOwner))
    (:then
      (bind _output_get (Variables|Default|SetPilgrimQuickplayPlayspace _asbp_pilgrim_quickplay_playspace))
      (bind _eventrouter (Pilgrim|PilgrimWaitForEventRouter _output_get)
        (:OnEventRouterAvailable
          (Variables|Default|SetEventRouter _eventrouter)
          (if (Variables|PilgrimPlayspaceComponent_IntroDirector|GetbEnableIntroDirector)
            (Networking|CanExecuteCosmeticEvents
              (:True
                ;[1 0.5575 0.142] Listen for the event that starts the intro cinematic
                (bind (_asynctaskproxy _payload) (GameplayEventRouter|StartListeningToEvent (Variables|Default|GetEventRouter) "/Script/PilgrimQuickplayRuntime.PilgrimEvent_StartIntro")
                  (:then
                    (Variables|EventHandles|SetIntroStartedEventHandle _asynctaskproxy)
                    (bind _eventrouter_2 (Pilgrim|PilgrimWaitForEventRouter (Components|GetOwner))
                      (:OnEventRouterAvailable
                        (Utilities|IsValid _eventrouter_2
                          (:Is%20Valid
                            (bind (_asynctaskproxy_2 _payload_2) (GameplayEventRouter|StartListeningToEvent _eventrouter_2 "/Script/PilgrimQPSharedRuntime.PilgrimEvent_QuickplayState_Ended")))))))
                  (:OnEventReceived
                    (Variables|Default|SetIntroPlayers (Utilities|Struct|BreakPilgrimEventStartIntro _payload))
                    (Utilities|FlowControl|Sequence
                      (:then_0
                        ;[0.9251 0.6278 1] Set up UI for intro/spotlight visuals
                        (for _array_element (Fortnite|GetLocalFortPlayerControllers true)
                          (DynamicUI|AddScene (WorldSubsystems|GetDynamicUIManager) "/PilgrimQuickplay/Widgets/Scenes/UIScene_PilgrimQuickplay_Intro.UIScene_PilgrimQuickplay_Intro" _array_element)))
                      (:then_1
                        ;[0 1 0.4643] Hide HUD elements for cinematic
                        (HUD|SetHUDElementVisibilitywithContext (Fortnite|GetFirstLocalFortPlayerController true) (Variables|Default|GetHideHUDElementsTags) true "ForPilgrimQuickplayIntro" ""))
                      (:then_2
                        (CallFunction|DetermineLocalTeam))
                      (:then_3
                        ;[0.124 1 0.6299] Ascynchronosly loading our spotlight emotes
                        (CallFunction|LoadPlayerSpotlightMontages))
                      (:then_4
                        ;[0.3281 1 0.9202] Start the intros!
                        (CallFunction|GotoState "NewEnumerator1"))))))) ; Intro sequence is client only
          ))))
    (:CastFailed
      (Development|PrintString "BP_PilgrimQuickplay_BandIntro_Director can only be placed on a pilgrim quickplay playspace"))))

(event EventOnSpotlightMontageLoaded
  (params (PlayerIndex :type Integer) (LoadedMontage :type Anim%20Montage%20Object%20Reference))
  ;[0.124 1 0.6299] Ascynchronosly loading our spotlight emotes
  (Utilities|Array|SetArrayElem (Variables|Default|GetPlayerSpotlightMontages) PlayerIndex LoadedMontage))

(event EventEndPlay
  (params (EndPlayReason :type EEndPlayReason%20Enum :default "Destroyed"))
  (bind _introstartedeventhandle (Variables|EventHandles|GetIntroStartedEventHandle))
  (Utilities|IsValid _introstartedeventhandle
    (:Is%20Valid
      (GameplayEventRouter|StopListeningtoEvent _introstartedeventhandle)
      (Variables|EventHandles|SetIntroStartedEventHandle 0))))

(event Custom|GotoNextState
  ;[0.3281 1 0.9202] Select the 'next' state to goto from the current one
  (switch Utilities|FlowControl|Switch|SwitchonBandIntroStagesEnum (Variables|Default|GetCurrentBandIntroState)
    (:NewEnumerator5
      _)
    (:NewEnumerator0
      _)
    (:NewEnumerator1
      (if (and (Pilgrim|QuickIntro|IsPilgrimQuickIntroEnabled) (not (Pilgrim|QuickIntro|ShouldQuickIntroKeepSpotlights)))
        (CallFunction|GotoState "NewEnumerator4")
        (else
          (CallFunction|GotoState "NewEnumerator3"))))
    (:NewEnumerator2
      (CallFunction|GotoState "NewEnumerator3"))
    (:NewEnumerator3
      (CallFunction|GotoState "NewEnumerator4")))
  (CallFunction|GotoState "NewEnumerator1"))

(event Custom|GotoState
  (params (NewState :type BandIntroStagesEnum%20Enum :default "NewEnumerator5"))
  ;[0.8723 0.09279 1] Goto the next state if 'None' is the new state
  (if (Utilities|Enum|NotEqual(Enum) NewState "NewEnumerator5")
    ;[1 0.8905 0] Record the current state as the last state before updating
    (Variables|Default|SetLastBandIntroState (Variables|Default|GetCurrentBandIntroState))
    ;[0.6569 0.4534 1] record the new state and act on it
    (bind _output_get (Variables|Default|SetCurrentBandIntroState NewState))
    (switch Utilities|FlowControl|Switch|SwitchonBandIntroStagesEnum _output_get
      (:NewEnumerator5
        _)
      (:NewEnumerator0
        _)
      (:NewEnumerator1
        (CallFunction|PlayGroupIntro))
      (:NewEnumerator2
        _)
      (:NewEnumerator3
        (CallFunction|PlaySpotlightIntros))
      (:NewEnumerator4
        (CallFunction|EndBandIntro)))
    (CallFunction|GotoNextState)
    (else
      (CallFunction|GotoNextState))))

(event Custom|PlayGroupIntro
  (bind _pilgrimquickplayplayspace (Variables|Default|GetPilgrimQuickplayPlayspace))
  ;[0.4072 0.1037 0.3919] Start the intro cinematic by playing the Band Intro camera
  (Class|SparksCameraDirectorComponent|HandleDirectorDynamicEvent (Variables|/Script/SparksMusicPlayspaceRuntime.SparksMusicPlayspace|GetCameraDirector _pilgrimquickplayplayspace) (Utilities|Struct|MakeSparksCommonDynamicCameraDirectorEvent false true -1.0 false 0 0 true (select (Pilgrim|QuickIntro|IsPilgrimQuickIntroEnabled) "(GameplayTags=((TagName=\"CameraDirector.ShotType.QuickBandIntro\")))" "(GameplayTags=((TagName=\"CameraDirector.ShotType.BandIntro\")))") (delegate Custom|BandIntroPrePlay) (delegate Custom|BandIntroFinished) (delegate Custom|BandIntroPostPlay)))
  (Class|PilgrimQuickplayPlayspace|ForceActiveCameraDirector _pilgrimquickplayplayspace) ; Called after so we start with dynamic camera
  (if (Pilgrim|QuickIntro|IsPilgrimQuickIntroEnabled)
    (Utilities|Time|SetTimerbyEvent (delegate Custom|OnQuickIntroTimerCompleted) (Pilgrim|QuickIntro|GetQuickIntroMinCameraSeconds))))

(event Custom|OnQuickIntroTimerCompleted
  ;[0.4072 0.1037 0.3919] Start the intro cinematic by playing the Band Intro camera
  (if (Utilities|Enum|Equal(Enum) (Variables|Default|GetCurrentBandIntroState) "NewEnumerator1")
    (CallFunction|GotoNextState)))

(event Custom|BandIntroPrePlay
  (params (Shot :type Sparks%20Camera%20Shot%20Structure%20%28by%20ref%29) (Sequence :type Level%20Sequence%20Actor%20Object%20Reference) (Target :type Actor%20Object%20Reference))
  ;[0.4072 0.1037 0.3919] Start the intro cinematic by playing the Band Intro camera
  (CallFunction|BindAllPawnstoSequence Sequence)
  (if (not (Pilgrim|QuickIntro|IsPilgrimQuickIntroEnabled))
    (Audio|PlaySound2D "/SparksMaps/Audio/FM_BandPortal.FM_BandPortal")))

(event Custom|BandIntroFinished
  (params (Shot :type Sparks%20Camera%20Shot%20Structure%20%28by%20ref%29))
  ;[0.4072 0.1037 0.3919] Start the intro cinematic by playing the Band Intro camera
  (if (not (Pilgrim|QuickIntro|IsPilgrimQuickIntroEnabled))
    (CallFunction|GotoNextState)))

(event Custom|BandIntroPostPlay
  (params (Shot :type Sparks%20Camera%20Shot%20Structure%20%28by%20ref%29) (Sequence :type Level%20Sequence%20Actor%20Object%20Reference) (Target :type Actor%20Object%20Reference))
  ;[0.4072 0.1037 0.3919] Start the intro cinematic by playing the Band Intro camera
  (for _array_element (Variables|Default|GetIntroPlayers)
    (Animation|Montage|MontageStop (Components|SkeletalMesh|GetAnimInstance (Variables|/Script/Engine.Character|GetMesh (.Pawn (Utilities|Struct|BreakPilgrimPlayerIntroInfo _array_element)))) 0.0 0)))

(event Custom|PlaySpotlightIntros
  ;[0.3281 1 0.9202] Begin playing the spotlight intros for each player present, starting with the player at index 0 in the Band Members array
  (bind _output_get (Variables|Default|SetCurrentSpotlightPlayerIndex 0))
  (CallFunction|PlaySpotlightforPlayerIndex _output_get))

(event Custom|PlaySpotlightForPlayerIndex
  (params (SpotlightPlayerIndex :type Integer))
  (bind _introplayers (Variables|Default|GetIntroPlayers))
  (bind (_currentintroplayer_playerstate _currentintroplayer_pawn) (Utilities|Struct|BreakPilgrimPlayerIntroInfo (Variables|GetCurrentIntroPlayer_PlayerState)))
  ;[0.8449 1 0] Play spotlight anim with end/completed event bound to call NextSpotlightPlayer
  (if (Utilities|Array|IsValidIndex _introplayers SpotlightPlayerIndex)
    (bind _output_get (Variables|Default|SetCurrentIntroPlayer (Utilities|Array|Get(acopy) _introplayers SpotlightPlayerIndex)))
    (Utilities|IsValid (Utilities|Struct|BreakPilgrimPlayerIntroInfo _output_get)
      (:Is%20Valid
        ;;[0.8449 1 0] Only play spotlights for players on my team (relevant for Arena)
        (if (== (Team|GetActorTeam (Utilities|Struct|BreakPilgrimPlayerIntroInfo _output_get)) (Variables|Default|GetLocalTeam))
          (CallFunction|CreateIntroOverlayforPlayer (Utilities|Struct|BreakPilgrimPlayerIntroInfo _output_get))
          ;;[0 0.1387 1] Looping until we get montage.  But we can't really delay here, so this is weird.  We should probably force the issue instead with a synchronous load.
          (bind _returnvalue (CallFunction|HasSpotlightMontageLoadFinished SpotlightPlayerIndex))
          (if (not _returnvalue)
            (Utilities|FlowControl|Delay 0.0))
          (Utilities|IsValid (Utilities|Array|Get(acopy) (Variables|Default|GetPlayerSpotlightMontages) SpotlightPlayerIndex)
            (:Is%20Valid
              (Utilities|IsValid _currentintroplayer_pawn
                (:Is%20Valid
                  (Class|SparksCameraDirectorComponent|HandleDirectorDynamicEvent (Variables|/Script/SparksMusicPlayspaceRuntime.SparksMusicPlayspace|GetCameraDirector (Variables|Default|GetPilgrimQuickplayPlayspace)) (Utilities|Struct|MakeSparksCommonDynamicCameraDirectorEvent false true -1.0 false _currentintroplayer_pawn 0 true "(GameplayTags=((TagName=\"CameraDirector.ShotType.Spotlight\")))" (delegate Custom|SpotlightPrePlay))))
                (:Is%20Not%20Valid
                  _)))
            (:Is%20Not%20Valid
              _))
          (CallFunction|SpotlightEnded)
          (else
            _)))
      (:Is%20Not%20Valid
        _))
    (CallFunction|NextSpotlightPlayer)
    (else
      (Development|PrintString "Invalid Spotlight Player Index"))))

(event Custom|SpotlightPrePlay
  (params (Shot :type Sparks%20Camera%20Shot%20Structure%20%28by%20ref%29) (Sequence :type Level%20Sequence%20Actor%20Object%20Reference) (Target :type Actor%20Object%20Reference))
  ;[0.8449 1 0] Play spotlight anim with end/completed event bound to call NextSpotlightPlayer
  (bind _asfort_player_pawn (Utilities|Casting|CastToFortPlayerPawn Target)
    (:then
      _)
    (:CastFailed
      _))
  (CallFunction|SetWeaponVisibility false _asfort_player_pawn)
  (Animation|Montage|PlayMontage (Variables|/Script/Engine.Character|GetMesh _asfort_player_pawn) (Utilities|Array|Get(acopy) (Variables|Default|GetPlayerSpotlightMontages) (stale Custom|PlaySpotlightForPlayerIndex SpotlightPlayerIndex))
    (:OnCompleted
      _)
    (:OnInterrupted
      _))
  (CallFunction|SpotlightEnded) ; This will indirectly end the camera shot
  (CallFunction|SetWeaponVisibility true _asfort_player_pawn))

(event Custom|CreateIntroOverlayForPlayer
  (params (PlayerState :type Player%20State%20Object%20Reference))
  ;[0.8449 1 0] Play spotlight anim with end/completed event bound to call NextSpotlightPlayer
  ;;[1 0.07087 0.3752] Create an overlay widget with data for this player and add it to the viewport
  (CallFunction|StartPlayerIntro PlayerState))

(event Custom|SpotlightEnded
  ;[0.8449 1 0] Play spotlight anim with end/completed event bound to call NextSpotlightPlayer
  ;[1 0.5575 0.142] Go to the next spotlight player
  (CallFunction|NextSpotlightPlayer))

(event Custom|NextSpotlightPlayer
  ;[1 0 0.9559] Cycle the current spotlight player index and play the spotlight for that player if there are any players left; otherwise, go to the next state
  (bind _result (Math|Integer|IncrementInt (Variables|Default|GetCurrentSpotlightPlayerIndex))
    (:%20%20
      (if (< _result (Utilities|Array|Length (Variables|Default|GetIntroPlayers)))
        (CallFunction|PlaySpotlightforPlayerIndex _result)
        (else
          (CallFunction|GotoNextState))))))

(event Custom|EndBandIntro
  (bind _eventrouter (Variables|Default|GetEventRouter))
  (Utilities|FlowControl|Sequence
    (:then_0
      ;[1 1 1] Indicate end of intro state for the pilgrim quickplay state machine (the state is currently advanced by a timer) - Server has to know about this!
      (Utilities|IsValid _eventrouter
        (:Is%20Valid
          (GameplayEventRouter|BroadcastEvent _eventrouter (Utilities|Struct|MakeSparksCommonCameraFinishedEvent) 0)
          ;;[1 1 1] Let the stage manager know it's time to change the mood
          (GameplayEventRouter|BroadcastEvent _eventrouter (Utilities|Struct|MakePilgrimEventQuickplayReturningtoGameplay) 0)
          (Class|SparksCameraDirectorComponent|Reset (Variables|/Script/SparksMusicPlayspaceRuntime.SparksMusicPlayspace|GetCameraDirector (Variables|Default|GetPilgrimQuickplayPlayspace)) true) ; Start playing intermission
        )))
    (:then_1
      ;[1 0.2083 0.9651] Stop hiding HUD elements we were hiding for the cinematic
      (HUD|SetHUDElementVisibilitywithContext (Fortnite|GetFirstLocalFortPlayerController true) (Variables|Default|GetHideHUDElementsTags) false "ForPilgrimQuickplayIntro" ""))
    (:then_2
      ;[0.1356 0.4072 0.3763] Ensure all players are visible - fix for a temp issue with invisible avatars
      (for _array_element (Variables|Default|GetIntroPlayers)
        (CallFunction|HidePlayerforCinematic (.Pawn (Utilities|Struct|BreakPilgrimPlayerIntroInfo _array_element)))))
    (:then_3
      (for _array_element_2 (Fortnite|GetLocalFortPlayerControllers true)
        (DynamicUI|RemoveScene (WorldSubsystems|GetDynamicUIManager) "/PilgrimQuickplay/Widgets/Scenes/UIScene_PilgrimQuickplay_Intro.UIScene_PilgrimQuickplay_Intro" _array_element_2)))
    (:then_4
      ;[1 0.1955 0.2552] Releasing Spotlight Montage assets that were loaded in, so they dont take up memory
      (CallFunction|ClearSpotlightMontageHandles)
      (Utilities|Array|Clear (Variables|Default|GetPlayerSpotlightMontages)))))

(event EventDebugShowSpotlightTool_Impl
  (params (bShow :type Boolean))
  (bind _debug_widget (Variables|Default|GetDebug%20Widget))
  (if bShow
    (Utilities|IsValid _debug_widget
      (:Is%20Not%20Valid
        (bind _returnvalue (UserInterface|CreateWidget "/PilgrimQuickplay/Utils/WBP_Pilgrim_Cosmetics_TestPanel.WBP_Pilgrim_Cosmetics_TestPanel_C"))
        (bind _output_get (Variables|Default|SetDebug%20Widget _returnvalue))
        (Utilities|IsValid _output_get
          (:Is%20Valid
            (UserInterface|Viewport|AddtoViewport _returnvalue 0)
            (Class|WBPPilgrimCosmeticsTestPanel|SetBandIntroDirector _returnvalue self :!class /PilgrimQuickplay/Utils/WBP_Pilgrim_Cosmetics_TestPanel.WBP_Pilgrim_Cosmetics_TestPanel_C)))))
    (else
      (Utilities|IsValid _debug_widget
        (:Is%20Valid
          (Widget|RemovefromParent _debug_widget)
          (Variables|Default|SetDebug%20Widget 0))))))

(fn BindAllPawnsToSequence
  (params (SequenceActor :type Level%20Sequence%20Actor%20Object%20Reference))
  ;[0.5491 1 0.9502] Bind individual pawns by individual tag, e.g. Pawn1, Pawn2
  (bind (_array_element _array_index) (Utilities|Array|ForEachLoopwithBreak (Variables|Default|GetIntroPlayers))
    (:LoopBody
      (if (>= _array_index 4)
        _)
      (Utilities|Array|Add (LocalVariables|GetAllPawns) (.Pawn (Utilities|Struct|BreakPilgrimPlayerIntroInfo _array_element)))
      (Sequencer|Player|Bindings|SetBindingbyTag (Variables|GetSequenceActor) (Utilities|Select :Index _array_index :Option%200 "Pawn1" :Option%201 "Pawn2" :Option%202 "Pawn3" :Option%203 "Pawn4") (Utilities|Array|MakeArray (.Pawn (Utilities|Struct|BreakPilgrimPlayerIntroInfo _array_element))) false))
    (:Completed
      _))
  ;[1 0.5943 0.9197] Bind all pawns to the group track, e.g. events that tell them all to show/hide or play montage
  (Sequencer|Player|Bindings|SetBindingbyTag SequenceActor (Utilities|Name|MakeLiteralName "AllPawns") (LocalVariables|GetAllPawns) false))

(fn CanApplyHolster
  (params (HolsterId :type Name%20%28by%20ref%29 :default "None"))
  (return true))

(fn DebugPlayerPos
  (params (context :type String))
  (bind (_output_playerstate _output_pawn) (Utilities|Struct|BreakPilgrimPlayerIntroInfo (Utilities|Array|Get(aref) (Variables|Default|GetIntroPlayers))))
  (Development|PrintString (Utilities|String|Append "IDir: player 0 location: " (Variables|Getcontext) " : " (Utilities|String|ToString(Vector) (Transformation|GetActorLocation _output_pawn)))))

(fn DetermineLocalTeam
  (Variables|Default|SetLocalTeam (Team|GetActorTeam (Fortnite|GetFirstLocalFortPlayerController true))))

(fn FindGroundedPawnLocation
  (params (Pawn :type Fort%20Player%20Pawn%20Object%20Reference))
  (bind _returnvalue (Transformation|GetActorTransform Pawn))
  (bind (_outhit _returnvalue_2) (Collision|LineTraceByChannel :Start (.location _returnvalue) :End (Utilities|Operators|Add (.location _returnvalue) (0.0 0.0 -1e+03)) :TraceChannel "TraceTypeQuery13" :bTraceComplex false :DrawDebugType "None" :bIgnoreSelf true :TraceColor "(R=1.000000,G=0.000000,B=0.000000,A=1.000000)" :TraceHitColor "(R=0.000000,G=1.000000,B=0.000000,A=1.000000)" :DrawTime 25.0))
  (if _returnvalue_2
    (return (Math|Transform|MakeTransform (Math|Vector|MakeVector (.x (.ImpactPoint (Collision|BreakHitResult _outhit))) (.y (.ImpactPoint (Collision|BreakHitResult _outhit))) (+ (.z (.ImpactPoint (Collision|BreakHitResult _outhit))) (Components|Capsule|GetScaledCapsuleHalfHeight (Variables|/Script/Engine.Character|GetCapsuleComponent (Variables|GetPawn))))) (.rotation _returnvalue) (.scale _returnvalue)))
    (else
      (return _returnvalue))))

(fn GetCurrentSpotlightPlayerIndex
  (return (Variables|Default|GetCurrentSpotlightPlayerIndex)))

(fn GetPawnAtIndex
  (params (Slot :type Integer))
  (bind _introplayers (Variables|Default|GetIntroPlayers))
  (bind (_output_playerstate _output_pawn) (Utilities|Struct|BreakPilgrimPlayerIntroInfo (Utilities|Array|Get(acopy) _introplayers Slot)))
  (if (Utilities|Array|IsValidIndex _introplayers Slot)
    (return _output_pawn)
    (else
      (return 0))))

(fn GetPlayerIndexForPawn
  (params (Pawn :type Pawn%20Object%20Reference))
  (for (_array_element _array_index) (Variables|Default|GetIntroPlayers)
    (if (== (.Pawn (Utilities|Struct|BreakPilgrimPlayerIntroInfo _array_element)) Pawn)
      (return _array_index true)))
  (return -1 false))

(fn HidePlayerForCinematic
  (params (PlayerPawn :type Pawn%20Object%20Reference) (ShouldHide :type Boolean))
  (bind _rootcomponent (Variables|/Script/Engine.Actor|GetRootComponent PlayerPawn))
  (Utilities|IsValid PlayerPawn
    (:Is%20Valid
      (Utilities|IsValid _rootcomponent
        (:Is%20Valid
          (Rendering|SetVisibility _rootcomponent (not ShouldHide) true)
          (return))))))

(fn IsValidGameStateToRestoreStartLocation
  (return (Utilities|Enum|Equal(Enum) (Variables|Default|GetCurrentQuickplayState) "SongGameplay")))

(fn LoadPlayerSpotlightMontages
  (Utilities|Array|Resize (Variables|Default|GetPlayerSpotlightMontages) 4)
  (for (_array_element _array_index) (Variables|Default|GetIntroPlayers)
    (Utilities|IsValid (.Pawn (Utilities|Struct|BreakPilgrimPlayerIntroInfo _array_element))
      (:Is%20Valid
        (Utilities|IsValid (Variables|/Script/FortniteGame.FortPawn|GetCurrentWeapon (.Pawn (Utilities|Struct|BreakPilgrimPlayerIntroInfo _array_element)))
          (:Is%20Valid
            (Utilities|IsValid (Actor|GetComponentbyClass (Variables|/Script/FortniteGame.FortPawn|GetCurrentWeapon (.Pawn (Utilities|Struct|BreakPilgrimPlayerIntroInfo _array_element))) "/SparksCosmetics/Components/BP_SparksInstrumentSettingsComponent.BP_SparksInstrumentSettingsComponent_C")
              (:Is%20Valid
                (CallFunction|RequestNextMontageforInstrument (Class|SparksInstrumentComponentBase|GetInstrumentType (Actor|GetComponentbyClass (Variables|/Script/FortniteGame.FortPawn|GetCurrentWeapon (.Pawn (Utilities|Struct|BreakPilgrimPlayerIntroInfo _array_element))) "/SparksCosmetics/Components/BP_SparksInstrumentSettingsComponent.BP_SparksInstrumentSettingsComponent_C")) (.Pawn (Utilities|Struct|BreakPilgrimPlayerIntroInfo _array_element)) _array_index))
              (:Is%20Not%20Valid
                _)))
          (:Is%20Not%20Valid
            _))
        (Fortnite|FortShippingLog 0 "Intro Band Director: Couldn't load Monatge"))
      (:Is%20Not%20Valid
        _))))

(fn PlayIntroMontageByInstrument
  (params (Actor :type Actor%20Object%20Reference) (In%20Portal :type Boolean))
  (bind _playerintrosbyinstrumenttype (Variables|Default|GetPlayerIntrosByInstrumentType))
  (bind _in_portal (Variables|GetIn%20Portal))
  (bind _asfort_player_pawn (Utilities|Casting|CastToFortPlayerPawn Actor))
  (Utilities|IsValid (Variables|/Script/FortniteGame.FortPawn|GetCurrentWeapon _asfort_player_pawn)
    (:Is%20Valid
      ;[1 0.5506 0.8947] Find the band position (index) for this pawn
      (bind (_playerindex _foundpawn) (CallFunction|GetPlayerIndexforPawn _asfort_player_pawn))
      (if _foundpawn
        ;[0.6798 1 0.5604] Find the appropriate data by instrument type and band position
        (if (.ReturnValue (Utilities|Map|Find _playerintrosbyinstrumenttype (Variables|/SparksCosmetics/Components/BP_SparksInstrumentSettingsComponent.BP_SparksInstrumentSettingsComponent_C|GetInstrumentType (Actor|GetComponentbyClass (Variables|/Script/FortniteGame.FortPawn|GetCurrentWeapon _asfort_player_pawn) "/SparksCosmetics/Components/BP_SparksInstrumentSettingsComponent.BP_SparksInstrumentSettingsComponent_C"))))
          (if (Utilities|Array|IsValidIndex (Utilities|Struct|BreakPilgrimBandIntroSequenceSet (Utilities|Map|Find _playerintrosbyinstrumenttype (Variables|/SparksCosmetics/Components/BP_SparksInstrumentSettingsComponent.BP_SparksInstrumentSettingsComponent_C|GetInstrumentType (Actor|GetComponentbyClass (Variables|/Script/FortniteGame.FortPawn|GetCurrentWeapon _asfort_player_pawn) "/SparksCosmetics/Components/BP_SparksInstrumentSettingsComponent.BP_SparksInstrumentSettingsComponent_C")))) _playerindex)
            ;[0.5311 0.9901 1] Select Air or Land part
            (Utilities|IsValid (select _in_portal (Utilities|Struct|BreakPilgrimBandIntroAnimParts (Utilities|Array|Get(aref) (Utilities|Struct|BreakPilgrimBandIntroSequenceSet (Utilities|Map|Find _playerintrosbyinstrumenttype (Variables|/SparksCosmetics/Components/BP_SparksInstrumentSettingsComponent.BP_SparksInstrumentSettingsComponent_C|GetInstrumentType (Actor|GetComponentbyClass (Variables|/Script/FortniteGame.FortPawn|GetCurrentWeapon _asfort_player_pawn) "/SparksCosmetics/Components/BP_SparksInstrumentSettingsComponent.BP_SparksInstrumentSettingsComponent_C")))) _playerindex)) (.StageAnim_8_3AA67D1542A02B7BF7C5EDA730C84964 (Utilities|Struct|BreakPilgrimBandIntroAnimParts (Utilities|Array|Get(aref) (Utilities|Struct|BreakPilgrimBandIntroSequenceSet (Utilities|Map|Find _playerintrosbyinstrumenttype (Variables|/SparksCosmetics/Components/BP_SparksInstrumentSettingsComponent.BP_SparksInstrumentSettingsComponent_C|GetInstrumentType (Actor|GetComponentbyClass (Variables|/Script/FortniteGame.FortPawn|GetCurrentWeapon _asfort_player_pawn) "/SparksCosmetics/Components/BP_SparksInstrumentSettingsComponent.BP_SparksInstrumentSettingsComponent_C")))) _playerindex))))
              (:Is%20Valid
                ;[1 0.4643 0.3543] Play the montage
                (Animation|Montage|MontagePlay (Components|SkeletalMesh|GetAnimInstance (Variables|/Script/Engine.Character|GetMesh _asfort_player_pawn)) (select _in_portal (Utilities|Struct|BreakPilgrimBandIntroAnimParts (Utilities|Array|Get(aref) (Utilities|Struct|BreakPilgrimBandIntroSequenceSet (Utilities|Map|Find _playerintrosbyinstrumenttype (Variables|/SparksCosmetics/Components/BP_SparksInstrumentSettingsComponent.BP_SparksInstrumentSettingsComponent_C|GetInstrumentType (Actor|GetComponentbyClass (Variables|/Script/FortniteGame.FortPawn|GetCurrentWeapon _asfort_player_pawn) "/SparksCosmetics/Components/BP_SparksInstrumentSettingsComponent.BP_SparksInstrumentSettingsComponent_C")))) _playerindex)) (.StageAnim_8_3AA67D1542A02B7BF7C5EDA730C84964 (Utilities|Struct|BreakPilgrimBandIntroAnimParts (Utilities|Array|Get(aref) (Utilities|Struct|BreakPilgrimBandIntroSequenceSet (Utilities|Map|Find _playerintrosbyinstrumenttype (Variables|/SparksCosmetics/Components/BP_SparksInstrumentSettingsComponent.BP_SparksInstrumentSettingsComponent_C|GetInstrumentType (Actor|GetComponentbyClass (Variables|/Script/FortniteGame.FortPawn|GetCurrentWeapon _asfort_player_pawn) "/SparksCosmetics/Components/BP_SparksInstrumentSettingsComponent.BP_SparksInstrumentSettingsComponent_C")))) _playerindex)))) 1.0 "MontageLength" 0.0 true))
              (:Is%20Not%20Valid
                _))
            (else
              ;[0.5365 0.02488 0] Give info about what was missing in data
              (Development|PrintString (Utilities|String|Append "IDir: Intro Montage failed lookup for position " (Utilities|String|ToString(Integer) _playerindex) " in instrument " (Utilities|String|EnumtoString (Class|PilgrimCoreUtilsLibrary|GetHeldInstrumentTrackType _asfort_player_pawn))))))
          ;[0.03646 0.003418 0.003418] debug
          (Development|PrintString (Utilities|String|Append "IDir: played default intro montage"))
          (else
            (Development|PrintString (Utilities|String|Append "IDir: Intro Montage failed instrument type lookup: " (Utilities|String|EnumtoString (Class|PilgrimCoreUtilsLibrary|GetHeldInstrumentTrackType _asfort_player_pawn))))))))))

(fn PlayMontageonCurrentSpotlightPawn
  (params (Montage :type Anim%20Montage%20Object%20Reference))
  (bind (_currentintroplayer_playerstate _currentintroplayer_pawn) (Utilities|Struct|BreakPilgrimPlayerIntroInfo (Variables|GetCurrentIntroPlayer_PlayerState)))
  (Utilities|IsValid _currentintroplayer_pawn
    (:Is%20Valid
      (bind _returnvalue (Animation|Montage|MontagePlay (Components|SkeletalMesh|GetAnimInstance (Variables|/Script/Engine.Character|GetMesh _currentintroplayer_pawn)) Montage 1.0 "MontageLength" 0.0 true)))))

(fn RestoreOneStartLocation
  (params (BandMemberIndex :type Integer))
  (bind _introplayers (Variables|Default|GetIntroPlayers))
  (bind (_output_playerstate _output_pawn _output_spawn) (Utilities|Struct|BreakPilgrimPlayerIntroInfo (Utilities|Array|Get(acopy) _introplayers BandMemberIndex)))
  (bind (_playerstart _transform) (Utilities|Struct|BreakPilgrimTeleportDestination _output_spawn))
  (Utilities|IsValid _output_pawn
    (:Is%20Valid
      (Transformation|SetActorTransform _output_pawn _transform false false))))

(fn SetSpotlightIndexLock
  (params (Index :type Integer))
  (Variables|PilgrimPlayspaceComponent_IntroDirector|SetSpotlightIndexLock Index))

(fn SetWeaponVisibility
  (params (Visible :type Boolean) (Pawn :type Fort%20Player%20Pawn%20Object%20Reference))
  (bind _returnvalue (Weapon|GetCurrentWeapon Pawn))
  (Utilities|IsValid _returnvalue
    (:Is%20Valid
      (FortWeapon|SetWeaponVisibility _returnvalue Visible true))))

(fn TeleportPawnatIndextoSpawnLocation
  (params (Slot%20Index :type Integer))
  (CallFunction|RestoreOneStartLocation Slot%20Index))
