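; Pilgrim Quickplay music playspace Blueprint that manages player restrictions, setlist limits, camera director and music mix across game states.
; Event router listeners and user join/leave hooks add or clear loose gameplay tags that block jam emotes and instrument selection.
; Connected player count sets songs allowed per player while the setlist is unlocked.
; CVars can close the emote picker and delay the jam-blocking tag to avoid hitches.
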(event EventBeginPlay
  (bind _self self)
  (bind _returnvalue (CallFunction|GetOwnedEventRouter _self))
  ;[1 1 1] If the player count changes, change the max songs that each player can add
  (Playspace|AssignNotifyPlayspaceUserAdded (delegate Custom|NotifyPlayspaceUserAdded_Event))
  (Playspace|AssignNotifyPlayspaceUserRemoved (delegate Custom|NotifyPlayspaceUserRemoved_Event))
  ;[1 1 1] Also set the inital max songs, in case players are already added before we BeginPlay
  (CallFunction|RefreshMaxCustomSetlistSongsPerPlayer)
  (bind (_asynctaskproxy _payload) (GameplayEventRouter|StartListeningToEvent _returnvalue "/Script/SparksCameraDirectorRuntime.SparksCommonStartCameraDirectorEvent")
    (:then
      (bind (_asynctaskproxy_2 _payload_2) (GameplayEventRouter|StartListeningToEvent _returnvalue "/Script/SparksCameraDirectorRuntime.SparksCommonStopCameraDirectorEvent")
        (:then
          (GameplayEventRouter|StartListeningToEvent _returnvalue "/Script/PilgrimQuickplayRuntime.PilgrimQuickplayAllPlayersReadyEvent"
            (:then
              (bind (_asynctaskproxy_3 _payload_3) (GameplayEventRouter|StartListeningToEvent _returnvalue "/Script/PilgrimQPSharedRuntime.PilgrimEvent_QuickplayState_BeganServer")
                (:then
                  (bind (_asynctaskproxy_4 _payload_4) (GameplayEventRouter|StartListeningToEvent _returnvalue "/Script/PilgrimCoreRuntime.PilgrimEvent_BandState_ReadinessChanged")
                    (:then
                      (bind (_asynctaskproxy_5 _payload_5) (GameplayEventRouter|StartListeningToEvent _returnvalue "/Script/PilgrimQPSharedRuntime.PilgrimEvent_QuickplayState_Began")
                        (:OnEventReceived
                          (CallFunction|EventHandleStateBeganEvent _payload_5))))
                    (:OnEventReceived
                      (CallFunction|EventHandlePlayerReadinessChangedEvent _payload_4))))
                (:OnEventReceived
                  (CallFunction|EventHandleStateBeganServerEvent _payload_3))))
            (:OnEventReceived
              (Class|SparksCameraDirectorComponent|ActivateCameraControl (Variables|SparksMusicPlayspace|GetCameraDirector) (Utilities|Select :Index (CallFunction|GetActivePilgrimStateId) :Loading false :Pregame false :Preintro false :Intro false :SongGameplay true :Outro false :SongResults false :SetResults false)))))
        (:OnEventReceived
          (CallFunction|HandleDirectorStopEvent _payload_2))))
    (:OnEventReceived
      (CallFunction|HandleDirectorStartEvent _payload))))

(event EventTick
  (params (DeltaSeconds :type Float%20%28single-precision%29)))

(event Custom|NotifyPlayspaceUserAdded_Event
  (params (AddedUser :type Playspace%20User%20Structure%20%28by%20ref%29))
  ;[1 1 1] If the player count changes, change the max songs that each player can add
  (CallFunction|RefreshMaxCustomSetlistSongsPerPlayer))

(event Custom|NotifyPlayspaceUserRemoved_Event
  (params (RemovedUser :type Playspace%20User%20Structure%20%28by%20ref%29))
  ;[1 1 1] If the player count changes, change the max songs that each player can add
  (CallFunction|RefreshMaxCustomSetlistSongsPerPlayer))

(event OnLockedStateChanged(PilgrimCustomSetlist)
  (params (Setlist :type Pilgrim%20Custom%20Setlist%20Object%20Reference) (bIsLocked :type Boolean))
  ;[1 1 1] Also change the max songs whenever the setlist is unlocked. This covers the case where a player drops mid-gameplay, and then we return to the green room.
  (CallFunction|RefreshMaxCustomSetlistSongsPerPlayer))

(event EventHandleDirectorResetEvent
  (params (Payload :type Sparks%20Common%20Reset%20Camera%20Director%20Event%20Structure%20%28by%20ref%29))
  (Class|SparksCameraDirectorComponent|Reset (Variables|SparksMusicPlayspace|GetCameraDirector) true))

(event EventHandleDirectorStartEvent
  (params (Payload :type Sparks%20Common%20Start%20Camera%20Director%20Event%20Structure%20%28by%20ref%29))
  (Class|SparksCameraDirectorComponent|ActivateCameraControl (Variables|SparksMusicPlayspace|GetCameraDirector) true))

(event EventHandleDirectorStopEvent
  (params (Payload :type Sparks%20Common%20Stop%20Camera%20Director%20Event%20Structure%20%28by%20ref%29))
  (Class|SparksCameraDirectorComponent|ActivateCameraControl (Variables|SparksMusicPlayspace|GetCameraDirector) false))

(event Custom|Event%20Handle%20State%20Began%20Server%20Event
  (params (Payload :type Pilgrim%20Event%20Quickplay%20State%20Began%20Server%20Structure))
  (switch Utilities|FlowControl|Switch|SwitchonEPilgrimQuickplayState (Utilities|Struct|BreakPilgrimEventQuickplayStateBeganServer Payload)
    (:Pregame
      (CallFunction|AllowJamEmotesforAllPlayers)
      (Default|AllowInstrumentSelectionforAllPlayers))
    (:Preintro
      _)
    (:Intro
      _)
    (:SongGameplay
      _)
    (:Outro
      _)
    (:SongResults
      (Utilities|FlowControl|Sequence :!label _label
        (:then_0
          (CallFunction|BlockJamEmotesforAllPlayers))
        (:then_1
          (Default|AllowInstrumentSelectionforAllPlayers))))
    (:SetResults
      (goto _label)))
  (CallFunction|BlockJamEmotesforAllPlayers)
  (CallFunction|BlockInstrumentSelectionforAllPlayers))

(event Custom|Event%20Handle%20State%20Began%20Event
  (params (Payload :type Pilgrim%20Event%20Quickplay%20State%20Began%20Structure))
  (Networking|CanExecuteCosmeticEvents
    (:True
      (switch Utilities|FlowControl|Switch|SwitchonEPilgrimQuickplayState (Utilities|Struct|BreakPilgrimEventQuickplayStateBase Payload)
        (:Pregame
          (CallFunction|TriggerControlBus false))
        (:Preintro
          _)
        (:Intro
          _)
        (:SongGameplay
          _)
        (:Outro
          _))
      (CallFunction|TriggerControlBus true))))

(event Custom|Event%20Handle%20Player%20Readiness%20Changed%20Event
  (params (Payload :type Pilgrim%20Event%20Band%20State%20Readiness%20Changed%20Structure))
  (if (Utilities|Struct|BreakPilgrimEventBandStateReadinessChanged Payload)
    (CallFunction|BlockJamEmotesforAllPlayers)
    (CallFunction|BlockInstrumentSelectionforAllPlayers)))

(event EventOnUserAdded
  (params (User :type Playspace%20User%20Structure%20%28by%20ref%29) (PlayerState :type Player%20State%20Object%20Reference))
  (|CallParentFunction User PlayerState)
  (switch Utilities|FlowControl|Switch|SwitchonEPilgrimQuickplayState (CallFunction|GetActivePilgrimStateId)
    (:Pregame
      (CallFunction|AllowJamEmotesforPlayer PlayerState)
      (Default|AllowInstrumentSelectionforPlayer 0)
      (CallFunction|TriggerControlBus false))
    (:Preintro
      _)
    (:Intro
      _)
    (:SongGameplay
      _)
    (:Outro
      _)
    (:SongResults
      (Utilities|FlowControl|Sequence :!label _label
        (:then_0
          (CallFunction|BlockJamEmotesforPlayer PlayerState))
        (:then_1
          (Default|AllowInstrumentSelectionforPlayer PlayerState))))
    (:SetResults
      (goto _label)))
  (CallFunction|BlockJamEmotesforPlayer PlayerState)
  (CallFunction|BlockInstrumentSelectionforPlayer 0)
  (CallFunction|TriggerControlBus true))

(event EventOnUserRemoved
  (params (User :type Playspace%20User%20Structure%20%28by%20ref%29) (PlayerState :type Player%20State%20Object%20Reference))
  (|CallParentFunction User PlayerState)
  (CallFunction|AllowJamEmotesforPlayer PlayerState)
  (Default|AllowInstrumentSelectionforPlayer PlayerState))

(event Custom|MultiCast_ForceResultCamera
  (bind _self self)
  (bind _returnvalue (CallFunction|GetOwnedEventRouter _self))
  (Networking|SwitchHasAuthority
    (:Remote
      (Utilities|IsValid _returnvalue
        (:Is%20Valid
          (GameplayEventRouter|BroadcastEvent _returnvalue (Utilities|Struct|MakeSparksCommonForceIntermissionCameraDirectorEvent "(GameplayTags=((TagName=\"CameraDirector.Pilgrim.Song.Ended\")))") 0))))))

(event Custom|DelayAddBlockingTag
  (params (PlayerState :type Player%20State%20Object%20Reference))
  (Utilities|FlowControl|DelayUntilNextTick)
  (CallFunction|AddBlockJamTagtoPlayer PlayerState))

(event EventEndPlay
  (params (EndPlayReason :type EEndPlayReason%20Enum :default "Destroyed"))
  (|CallParentFunction EndPlayReason)
  (Audio|Modulation|DeactivateControlBusMix (Variables|Default|GetMusicGameplayMix)))

(event Custom|TriggerControlBus
  (params (Activate :type Boolean))
  (bind _musicgameplaymix (Variables|Default|GetMusicGameplayMix))
  (bind _musicgameplaymix_2 (Variables|Default|GetMusicGameplayMix))
  (if Activate
    (if (not (Audio|Modulation|IsControlBusMixActive _musicgameplaymix))
      (Audio|Modulation|ActivateControlBusMix _musicgameplaymix))
    (elif (Audio|Modulation|IsControlBusMixActive _musicgameplaymix_2)
      (Audio|Modulation|DeactivateControlBusMix _musicgameplaymix_2))))

(fn AddBlockInstrumentSelectionTagToPlayer
  (params (PlayerState :type Player%20State%20Object%20Reference))
  (Ability|GameplayEffect|AddLooseGameplayTags PlayerState (GameplayTags|MakeGameplayTagContainerfromTag (Variables|Default|GetBlockInstrumentSelectionTag)) true)
  ;[1 0.5 0.5] The player now has the tag
  (return))

(fn AddBlockJamTagToPlayer
  (params (PlayerState :type Player%20State%20Object%20Reference))
  (Ability|GameplayEffect|AddLooseGameplayTags PlayerState (GameplayTags|MakeGameplayTagContainerfromTag (Variables|Default|GetBlockJamEmotesTag)) true)
  ;[1 0.5 0.5] The player now has the tag
  (return))

(fn AllowInstrumentSelectionForAllPlayers
  (for _array_element (CallFunction|GetAllPlayerStates)
    (Default|AllowInstrumentSelectionforPlayer _array_element))
  ;[1 0.5 0.5] Instrument selection now unblocked for all players
  (return))

(fn AllowInstrumentSelectionForPlayer
  (params (PlayerState :type Player%20State%20Object%20Reference))
  (CallFunction|ClearTagfromPlayer PlayerState (Variables|Default|GetBlockInstrumentSelectionTag))
  (return))

(fn AllowJamEmotesForAllPlayers
  (for _array_element (CallFunction|GetAllPlayerStates)
    (CallFunction|AllowJamEmotesforPlayer _array_element))
  ;[1 0.5 0.5] Jam emotes are now enabled for all players
  (return))

(fn AllowJamEmotesForPlayer
  (params (PlayerState :type Player%20State%20Object%20Reference))
  (CallFunction|ClearTagfromPlayer PlayerState (Variables|Default|GetBlockJamEmotesTag))
  (return))

(fn BlockInstrumentSelectionForAllPlayers
  (for _array_element (CallFunction|GetAllPlayerStates)
    (CallFunction|BlockInstrumentSelectionforPlayer _array_element))
  ;[1 0.5 0.5] Instrument selection has been blocked for all players
  (return))

(fn BlockInstrumentSelectionForPlayer
  (params (PlayerState :type Player%20State%20Object%20Reference))
  (bind _returnvalue (Ability|GetAbilitySystemComponent PlayerState))
  ;[1 1 1] Don't add the tag if it already exists
  (Utilities|IsValid PlayerState
    (:Is%20Valid
      (Utilities|IsValid _returnvalue
        (:Is%20Valid
          (if (GameplayTags|HasMatchingGameplayTag _returnvalue (Variables|Default|GetBlockInstrumentSelectionTag))
            ;[1 0.5 0.5] The player already has the tag
            (return)
            (else
              (CallFunction|AddBlockInstrumentSelectionTagtoPlayer PlayerState)
              ;[1 0.5 0.5] The player now has the tag
              (return))))
        (:Is%20Not%20Valid
          _)))
    (:Is%20Not%20Valid
      _))
  (Development|PrintString "Could not add Jam blocking tag")
  ;[1 0 0] This actor can't have the tag
  (return))

(fn BlockJamEmotesForAllPlayers
  (for _array_element (CallFunction|GetAllPlayerStates)
    (CallFunction|BlockJamEmotesforPlayer _array_element))
  ;[1 0.5 0.5] Jam has been blocked for all players
  (return))

(fn BlockJamEmotesForPlayer
  (params (PlayerState :type Player%20State%20Object%20Reference))
  (bind _returnvalue (Ability|GetAbilitySystemComponent PlayerState))
  (bind _returnvalue_2 (PlayerState|GetPlayerController PlayerState))
  (bind _returnvalue_3 (Actor|GetComponentbyClass _returnvalue_2 "/Script/PilgrimCoreRuntime.PilgrimPlayerControllerComponent"))
  ;[1 1 1] Don't add the tag if it already exists
  (Utilities|IsValid PlayerState
    (:Is%20Valid
      (Utilities|IsValid _returnvalue
        (:Is%20Valid
          (if (GameplayTags|HasMatchingGameplayTag _returnvalue (Variables|Default|GetBlockJamEmotesTag))
            ;[1 0.5 0.5] The player already has the tag
            (return)
            (elif (> (Fortnite|GetNumericCVar "Pilgrim.ShouldCloseEmotePickerBeforeBlockingJam") 0)
              ;[1 1 1] Close the picker to avoid a hitch
              (Utilities|IsValid _returnvalue_2
                (:Is%20Valid
                  (Utilities|IsValid _returnvalue_3
                    (:Is%20Valid
                      (Class|PilgrimPlayerControllerComponent|ServerRequestClientCloseEmotePicker _returnvalue_3))
                    (:Is%20Not%20Valid
                      _))
                  ;[1 1 1] Check if we should delay adding the tag
                  (if (> (Fortnite|GetNumericCVar "Pilgrim.ShouldDelayAddingJamBlockerTag") 0)
                    ;[1 1 1] Add the tag after a delay
                    (CallFunction|DelayAddBlockingTag PlayerState)
                    ;[1 0.5 0.5] The tag is set to be added
                    (return)
                    (else
                      _))))))
          ;[1 1 1] Check if we should delay adding the tag
          ;[1 1 1] Add the tag now
          (CallFunction|AddBlockJamTagtoPlayer PlayerState)
          ;[1 0.5 0.5] The player now has the tag
          (return))
        (:Is%20Not%20Valid
          _)))
    (:Is%20Not%20Valid
      _))
  (Development|PrintString "Could not add Jam blocking tag")
  ;[1 0 0] This actor can't have the tag
  (return))

(fn ClearTagFromPlayer
  (params (PlayerState :type Player%20State%20Object%20Reference) (GameplayTagToClear :type Gameplay%20Tag%20Structure))
  (bind _returnvalue (Ability|GetAbilitySystemComponent PlayerState))
  ;[1 1 1] Keep removing the instrument selection blocker tag for as many times as it has been added
  (Utilities|IsValid PlayerState
    (:Is%20Valid
      (Utilities|IsValid _returnvalue
        (:Is%20Valid
          (while (GameplayTags|HasMatchingGameplayTag _returnvalue GameplayTagToClear)
            (Ability|GameplayEffect|RemoveLooseGameplayTags PlayerState (GameplayTags|MakeGameplayTagContainerfromTag GameplayTagToClear) true))
          ;[1 0.5 0.5] The tag has been removed
          (return))
        (:Is%20Not%20Valid
          _)))
    (:Is%20Not%20Valid
      _))
  (Development|PrintString (Utilities|String|Append "Could not remove " (Utilities|String|ToString(Name) (GameplayTags|GetTagName GameplayTagToClear)) " tag"))
  ;[1 0.5 0.5] This actor can't have the tag
  (return))

(fn RefreshMaxCustomSetlistSongsPerPlayer
  (for _array_element (CallFunction|GetAllPlayerStates)
    ;[0.1144 0.15 0.1465] Only proceed if a player is connected
    (bind _asfort_player_state_athena (Utilities|Casting|CastToFortPlayerStateAthena _array_element))
    (if (not (State|IsDisconnected _asfort_player_state_athena))
      (Math|Integer|IncrementInt (LocalVariables|GetNumValidPlayersForSongCount))))
  ;[1 1 1] Now set the song count based on valid player count
  (CallFunction|SetMaxCustomSetlistSongsPerPlayer (LocalVariables|GetNumValidPlayersForSongCount)))

(fn SetMaxCustomSetlistSongsPerPlayer
  (params (NumPlayers :type Integer))
  (if (not (Variables|/Script/PilgrimCoreRuntime.PilgrimCustomSetlist|GetbIsSetlistLocked (Variables|Default|GetPilgrimCustomSetlist)))
    (Variables|/Script/PilgrimCoreRuntime.PilgrimCustomSetlist|SetMaxSongsPerPlayer (Utilities|Select :Index (Math|Integer|Clamp(Integer) NumPlayers 0 4) :Option%200 0 :Option%201 4 :Option%202 2 :Option%203 1 :Option%204 1) (Variables|Default|GetPilgrimCustomSetlist))))

(fn ConstructionScript)
