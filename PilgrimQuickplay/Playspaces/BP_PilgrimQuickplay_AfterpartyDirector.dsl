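; Server-side director for the Pilgrim Quickplay afterparty that triggers afterparty effects as the playspace re-enters the Pregame state.
; On BeginPlay with authority, it waits for the event router and listens for Quickplay state-began events, advancing the afterparty on each Pregame.
; When its pregame counter equals one, it finds and caches all BP_AfterpartyEntity actors.
; Each advance activates cached entities whose activation order matches the counter, then increments it.
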(event EventBeginPlay
  (Networking|SwitchHasAuthority
    (:Authority
      (bind _eventrouter (Pilgrim|PilgrimWaitForEventRouter (Components|GetOwner))
        (:OnEventRouterAvailable
          ;[0.5121 1 0.869] When we hit the pregame state, attempt to advance the afterparty
          (bind (_asynctaskproxy _payload) (GameplayEventRouter|StartListeningToEvent _eventrouter "/Script/PilgrimQPSharedRuntime.PilgrimEvent_QuickplayState_BeganServer")
            (:OnEventReceived
              (if (Utilities|Enum|Equal(Enum) (Utilities|Struct|BreakPilgrimEventQuickplayStateBegan _payload) "Pregame")
                (CallFunction|AdvanceAfterparty)))))))))

(event Custom|AdvanceAfterparty
  (Utilities|FlowControl|Sequence
    (:then_0
      ;[0.5121 1 0.869] If this is the start of the afterparty, find the actors first
      (if (== (Variables|Default|GetPregameStateBeganCount) 1)
        (Utilities|FlowControl|Sequence
          (:then_0
            ;[0.5121 1 0.869] Find all of the afterparty actors
            (CallFunction|FindAfterpartyActors))
          (:then_1
            _))
        (else
          _))
      (for _array_element (Variables|Default|GetAfterpartyActors)
        (if (== (Variables|/PilgrimQuickplay/Actors/Afterparty/BP_AfterpartyEntity.BP_AfterpartyEntity_C|GetAfterpartyActivationOrder _array_element) (Variables|Default|GetPregameStateBeganCount))
          (Class|BPAfterpartyEntity|ActivateAfterpartyEffect _array_element))))
    (:then_1
      ;[0.5121 1 0.869] Increment our count of how many times we've entered the pregame state
      (Math|Integer|IncrementInt (Variables|Default|GetPregameStateBeganCount)))))

(fn FindAfterpartyActors
  ;[0.5121 1 0.869] Find all AfterpartyEntity actors and cache them.
  (bind _outactors (Actor|GetAllActorsOfClass "/PilgrimQuickplay/Actors/Afterparty/BP_AfterpartyEntity.BP_AfterpartyEntity_C"))
  (Variables|Default|SetAfterpartyActors _outactors))
