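; Playspace component that decides whether the Pilgrim Quickplay loading screen stays up.
; It caches custom World Partition streaming sources, marks each done as streaming completes, and clears the cache when streaming state changes.
; The screen holds until all sources finish and the player is on the ground or warmup has started.
; It then spawns a shader warmup actor before the camera, destroying it on a later check to release the screen.
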(event EventBP_OnStreamingStateChanged
  (Variables|Default|SetHasRetrievedActors false)
  (Variables|Default|SetbAllSourcesDone false)
  (Utilities|Array|Clear (Variables|Default|GetCustomPartitionSources))
  (Development|PrintString "Clearing Custom Partition Sources" true true "(R=0.000000,G=0.660000,B=1.000000,A=1.000000)" 1.0))

(fn BP_ShouldDisplayLoadingScreen
  (Utilities|FlowControl|Sequence
    (:then_0
      ; Throttle "GetAllActors" logic
      (if (not (Variables|Default|GetHasRetrievedActors))
        (bind _outactors (Actor|GetAllActorsOfClass "/SparksMaps/SparksCustomWorldPartitionStreamingSource.SparksCustomWorldPartitionStreamingSource_C"))
        (for _array_element _outactors
          (Utilities|Array|AddUnique (Variables|Default|GetCustomPartitionSources) (Utilities|Struct|MakePilgrimCustomStreamingSourceStatus _array_element)))
        (Variables|Default|SetHasRetrievedActors (Utilities|Array|IsNotEmpty (Variables|Default|GetCustomPartitionSources)))))
    (:then_1
      (bind (_done _reason) (CallFunction|IsReady))
      (if (not _done)
        (LocalVariables|SetLoadingReason (Utilities|String|ToString(Text) _reason))
        (Variables|Default|SetbAllSourcesDone true)
        (for (_array_element_2 _array_index) (Variables|Default|GetCustomPartitionSources)
          (if (.Done_6_C883F0B24742E5AF310226AF6E4C3902 (Utilities|Struct|BreakPilgrimCustomStreamingSourceStatus _array_element_2))
            _
            (elif (and (Variables|/SparksMaps/SparksCustomWorldPartitionStreamingSource.SparksCustomWorldPartitionStreamingSource_C|GetbHasProcessedStreamingSources (Utilities|Struct|BreakPilgrimCustomStreamingSourceStatus _array_element_2)) (Streaming|IsStreamingCompleted (Variables|/SparksMaps/SparksCustomWorldPartitionStreamingSource.SparksCustomWorldPartitionStreamingSource_C|GetStreamingSource (Utilities|Struct|BreakPilgrimCustomStreamingSourceStatus _array_element_2)) :!class /Script/Engine.WorldPartitionStreamingSourceComponent))
              ; Set this streaming source to finished so we dont re-check it every time.
              (Utilities|Array|SetArrayElem (Variables|Default|GetCustomPartitionSources) _array_index (Utilities|Struct|MakePilgrimCustomStreamingSourceStatus (Utilities|Struct|BreakPilgrimCustomStreamingSourceStatus _array_element_2) true))
              (else
                (Variables|Default|SetbAllSourcesDone false)
                (LocalVariables|SetLoadingReason (Utilities|String|Append (LocalVariables|GetLoadingReason) "Pilgrim Source Still Pending : " (Utilities|GetDisplayName (Utilities|Struct|BreakPilgrimCustomStreamingSourceStatus _array_element_2))))))))
        (bind (_done_2 _reason_2) (CallFunction|IsReady))
        (if (not _done_2)
          (bind _output_get (LocalVariables|SetLoadingReason (Utilities|String|Append (LocalVariables|GetLoadingReason) (Utilities|String|ToString(Text) _reason_2))))
          ;[0.15 0.01386 0 0.5] Still processing streaming sources, keep loading screen open.
          (return _output_get true)))
      (if (== (Variables|Default|GetWarmupActorStartFrame) 0)
        (CallFunction|SpawnShaderWarmupActor)
        (return "" true)
        (else
          (CallFunction|DestroyShaderWarmupActor)
          (return "" false))))))

(fn DestroyShaderWarmupActor
  (if (Utilities|IsValid (Variables|Default|GetWarmupObjectReference) :!class /Script/Engine.KismetSystemLibrary)
    (Class|BPPilgrimWarmupCollection|VelocityPass (Variables|Default|GetWarmupObjectReference))
    (Actor|DestroyActor (Variables|Default|GetWarmupObjectReference))))

(fn IsReady
  (return (Math|Boolean|ANDBoolean (Variables|Default|GetbAllSourcesDone) (Variables|Default|GetHasRetrievedActors) (or (AI|Components|NavMovement|IsMovingonGround (Variables|/Script/Engine.Character|GetCharacterMovement (Game|GetPlayerCharacter 0))) (!= (Variables|Default|GetWarmupActorStartFrame) 0))) (Utilities|Text|FormatText "NSLOCTEXT(\"[4BB32AF0DDCF4CEA68809B5AC0E0BC62]\", \"C59D163A40622A98879BA889B24858CD\", \"Sources Done {SourcesDone}, HasRetreivedActors {HasRetreivedActors} IsMovingOnGround {IsMovingOnGround} WarmupActorStartFrame {WarmupActorStartFrame}\")" (Variables|Default|GetbAllSourcesDone) (Variables|Default|GetHasRetrievedActors) (AI|Components|NavMovement|IsMovingonGround (Variables|/Script/Engine.Character|GetCharacterMovement (Game|GetPlayerCharacter 0))) (!= (Variables|Default|GetWarmupActorStartFrame) 0))))

(fn SpawnShaderWarmupActor
  (bind (_outlocation _outrotation) (Actor|GetActorEyesViewPoint (Game|GetPlayerController 0)))
  ; Spawn the warmup actor in front of the player so it is rendered
  (if (not (Utilities|IsValid (Variables|Default|GetWarmupObjectReference) :!class /Script/Engine.KismetSystemLibrary))
    (Development|PrintString "Creating Pilgrim Warmup Collection")
    (bind _returnvalue (Game|SpawnActorfromClass "/PilgrimCore/Actors/BP_PilgrimWarmupCollection.BP_PilgrimWarmupCollection_C" (Math|Transform|MakeTransform (+ _outlocation (* (Math|Vector|GetForwardVector _outrotation) 1e+02)) (Math|Rotator|CombineRotators _outrotation (Math|Rotator|MakeRotator 9e+01 0.0 9e+01))) "AlwaysSpawn" "OverrideRootScale" (Components|GetOwner self) 0))
    (bind _output_get (Variables|Default|SetWarmupObjectReference _returnvalue))
    (Variables|Default|SetWarmupActorStartFrame (Utilities|GetFrameCount))
    (Class|BPPilgrimWarmupCollection|Warmup _output_get)))
