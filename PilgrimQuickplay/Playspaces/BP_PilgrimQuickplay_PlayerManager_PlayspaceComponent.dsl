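; Playspace component that manages backstage green room players, readiness and the countdown into gameplay for Pilgrim Quickplay.
; The server tracks team players, setlist votes and readiness through the gameplay event router, choosing timer states and checking tiebreakers.
; Clients update display board widgets from player events and handle ready-up input.
; Replicated timer state drives countdowns with curve-table lengths, rebroadcast as events, with waits for late routers and app-resume correction.
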(event EventBeginPlay
  (bind _greenroomplayerdisplay (Variables|Default|GetGreenRoomPlayerDisplay))
  (bind _bp_pilgrim_quickplay_playspace (Variables|Default|GetBP%20Pilgrim%20Quickplay%20Playspace))
  (Class|FortAsyncActionCurrentPlaylistReady|PlaylistDataReadyAsync
    (:then
      ;[0.5168 1 0.9401] Set the length of the backstage countdown timers from the PilgrimQuickplayGameData curve table
      (bind _outxy (DataTable|EvaluateCurveTableRow "/PilgrimQuickplay/DataTables/PilgrimQuickplayGameData.PilgrimQuickplayGameData" "Backstage.CountdownTimerLength.WaitingForPlayers")
        (:RowFound
          (Variables|Default|SetTimerLengthWhenPlayerUnfinished _outxy))
        (:RowNotFound
          _))
      (bind _outxy_2 (DataTable|EvaluateCurveTableRow "/PilgrimQuickplay/DataTables/PilgrimQuickplayGameData.PilgrimQuickplayGameData" "Backstage.CountdownTimerLength.WaitingForPlayers_Curated")
        (:RowFound
          (Variables|Default|SetTimerLengthWhenPlayerUnfinished_CuratedSetlist _outxy_2))
        (:RowNotFound
          _))
      (bind _outxy_3 (DataTable|EvaluateCurveTableRow "/PilgrimQuickplay/DataTables/PilgrimQuickplayGameData.PilgrimQuickplayGameData" "Backstage.CountdownTimerLength.AllPlayersReady")
        (:RowFound
          ;[0.02175 1 0] SERVER - Players joining and leaving events
          (Variables|Default|SetTimerLengthAllReady (select (Pilgrim|QuickIntro|IsPilgrimQuickIntroEnabled) 0.01 _outxy_3)))))
    (:OnReady
      (bind _asbp_pilgrim_quickplay_playspace (Utilities|Casting|CastToBP_PilgrimQuickplayPlayspace (Components|GetOwner))
        (:then
          (bind _output_get (Variables|Default|SetBP%20Pilgrim%20Quickplay%20Playspace _asbp_pilgrim_quickplay_playspace))
          (bind _eventrouter (Pilgrim|PilgrimWaitForEventRouter _output_get)
            (:OnEventRouterAvailable
              (Variables|Default|SetEventRouter _eventrouter)
              (bind _returnvalue (Actor|GetActorOfClass "/PilgrimQuickplay/Actors/GreenRoomPlayerDisplay/BP_GreenRoomPlayerDisplay.BP_GreenRoomPlayerDisplay_C"))
              (bind _output_get_2 (Variables|Default|SetGreenRoomPlayerDisplay _returnvalue))
              (Utilities|IsValid _output_get_2
                (:Is%20Valid
                  _)
                (:Is%20Not%20Valid
                  (Utilities|FlowControl|DelayUntilNextTick)))
              (Networking|SwitchHasAuthority
                (:Authority
                  (Utilities|FlowControl|Sequence
                    (:then_0
                      ;;[0.9161 1 0.156] Collect an initial set of info - before the listeners are set up
                      (for _array_element (Class|SparksMusicPlayspace|GetAllPlayerStates _bp_pilgrim_quickplay_playspace)
                        (CallFunction|ChecktoAddPlayerState _array_element)))
                    (:then_1
                      (Playspace|BindEventtoNotifyPlayspaceUserAdded _bp_pilgrim_quickplay_playspace (delegate Custom|NotifyPlayspaceUserAdded_Event))
                      (Playspace|BindEventtoNotifyPlayspaceUserRemoved _bp_pilgrim_quickplay_playspace (delegate Custom|NotifyPlayspaceUserRemoved_Event)))
                    (:then_2
                      (CallFunction|SetupCuratedSetlistInfo))
                    (:then_3
                      (CallFunction|SetupServerListeningEvents))))
                (:Remote
                  ;[0.4766 0.9502 1] Set the widgets to default before setting the reference to them - they are fair game to use the instant the reference array is set (if we get replicated data that we would set to the widget, but the reference is not ready, we will wait)
                  (bind _aswbp_green_room_display_board (Utilities|Casting|CastToWBP_GreenRoomDisplayBoard (UserInterface|GetWidget (Variables|/PilgrimQuickplay/Actors/GreenRoomPlayerDisplay/BP_GreenRoomPlayerDisplay.BP_GreenRoomPlayerDisplay_C|GetWidget _greenroomplayerdisplay))))
                  (bind _output_get_3 (Variables|/PilgrimQuickplay/Actors/GreenRoomPlayerDisplay/BP_GreenRoomPlayerDisplay.BP_GreenRoomPlayerDisplay_C|SetDisplayBoardWidget _aswbp_green_room_display_board _greenroomplayerdisplay))
                  (Class|WBPGreenRoomDisplayBoard|SetWidgetParent _output_get_3 (Variables|Default|GetGreenRoomPlayerDisplay) :!class /PilgrimQuickplay/Widgets/WBP_GreenRoomDisplayBoard.WBP_GreenRoomDisplayBoard_C)
                  (for _array_element_2 (Utilities|Array|MakeArray (Variables|/PilgrimQuickplay/Widgets/WBP_GreenRoomDisplayBoard.WBP_GreenRoomDisplayBoard_C|GetPlayer1_Widget _aswbp_green_room_display_board) (Variables|/PilgrimQuickplay/Widgets/WBP_GreenRoomDisplayBoard.WBP_GreenRoomDisplayBoard_C|GetPlayer2_Widget _aswbp_green_room_display_board) (Variables|/PilgrimQuickplay/Widgets/WBP_GreenRoomDisplayBoard.WBP_GreenRoomDisplayBoard_C|GetPlayer3_Widget _aswbp_green_room_display_board) (Variables|/PilgrimQuickplay/Widgets/WBP_GreenRoomDisplayBoard.WBP_GreenRoomDisplayBoard_C|GetPlayer4_Widget _aswbp_green_room_display_board))
                    (Class|WBPGreenRoomPlayerDisplay|SettoDefault _array_element_2))
                  (Variables|/PilgrimQuickplay/Actors/GreenRoomPlayerDisplay/BP_GreenRoomPlayerDisplay.BP_GreenRoomPlayerDisplay_C|SetGreenRoomPlayerDisplayWidgets (Utilities|Array|MakeArray (Variables|/PilgrimQuickplay/Widgets/WBP_GreenRoomDisplayBoard.WBP_GreenRoomDisplayBoard_C|GetPlayer1_Widget _aswbp_green_room_display_board) (Variables|/PilgrimQuickplay/Widgets/WBP_GreenRoomDisplayBoard.WBP_GreenRoomDisplayBoard_C|GetPlayer2_Widget _aswbp_green_room_display_board) (Variables|/PilgrimQuickplay/Widgets/WBP_GreenRoomDisplayBoard.WBP_GreenRoomDisplayBoard_C|GetPlayer3_Widget _aswbp_green_room_display_board) (Variables|/PilgrimQuickplay/Widgets/WBP_GreenRoomDisplayBoard.WBP_GreenRoomDisplayBoard_C|GetPlayer4_Widget _aswbp_green_room_display_board)) _greenroomplayerdisplay)
                  ;[0.4766 0.9502 1] Start listening to events on the client
                  (CallFunction|SetupClientListeningEvent))))))
        (:CastFailed
          (Development|PrintString "BP_PilgrimQuickplay_PlayerManager_PlayspaceComponent's parent failed cast to BP_PilgrimQuickplayPlayspace!" true true "(R=0.906250,G=0.195980,B=0.250125,A=1.000000)" 2e+01))))))

(event Custom|NotifyPlayspaceUserRemoved_Event
  (params (RemovedUser :type Playspace%20User%20Structure%20%28by%20ref%29))
  (bind (_userid _leafplayspace _playerstatecached) (Utilities|Struct|BreakPlayspaceUser RemovedUser))
  (bind _returnvalue (Utilities|Array|FindItem (Variables|Default|GetMyPlayerStates) _playerstatecached))
  (bind _myplayerstates (Variables|Default|GetMyPlayerStates))
  ;[0.02175 1 0] SERVER - Players joining and leaving events
  ;;[0.9779 1 0.147] Remove any folks that are leaving by setting their entry to nothing
  (Variables|Default|SetAllPlayerStates (Class|SparksMusicPlayspace|GetAllPlayerStates (Variables|Default|GetBP%20Pilgrim%20Quickplay%20Playspace)))
  (Utilities|FlowControl|Sequence
    (:then_0
      (if (> _returnvalue 0)
        (Utilities|Array|SetArrayElem _myplayerstates _returnvalue)
        (else
          (Development|PrintString "Player to remove not found!"))))
    (:then_1
      (Default|CheckForAllPlayersReady false))))

(event Custom|NotifyPlayspaceUserAdded_Event
  (params (AddedUser :type Playspace%20User%20Structure%20%28by%20ref%29))
  (bind (_userid _leafplayspace _playerstatecached) (Utilities|Struct|BreakPlayspaceUser AddedUser))
  ;[0.02175 1 0] SERVER - Players joining and leaving events
  (CallFunction|ChecktoAddPlayerState _playerstatecached))

(event Custom|CheckToAddPlayerState
  (params (Player%20State :type Player%20State%20Object%20Reference))
  (bind _myplayerstates (Variables|Default|GetMyPlayerStates))
  (bind _returnvalue (Utilities|Array|FindItem _myplayerstates))
  ;[0.02175 1 0] SERVER - Players joining and leaving events
  (Variables|Default|SetAllPlayerStates (Class|SparksMusicPlayspace|GetAllPlayerStates (Variables|Default|GetBP%20Pilgrim%20Quickplay%20Playspace)))
  (if (Utilities|Enum|Equal(Enum) (Utilities|Enum|BytetoEnumEFortTeam (Team|GetActorTeam Player%20State)) (Variables|Default|GetForTeamInArena))
    (CallFunction|MakeIndicatorVisibletoAll Player%20State)
    ;;[1 0.8602 0.07432] Check to see if the player is already represented
    (if (< (Utilities|Array|FindItem (Variables|Default|GetMyPlayerStates) Player%20State) 0)
      ;;[0.8236 1 0.1828] Find a free spot to add the player
      (if (> _returnvalue -1)
        (Utilities|Array|SetArrayElem _myplayerstates _returnvalue Player%20State)
        (else
          (Development|PrintString "No Free spots found!"))))))

(event Custom|MakeIndicatorVisibleToAll
  (params (InPlayerState :type Player%20State%20Object%20Reference))
  ;[1 0 0.9673] Add fort shipping log saying we're about to go into 'Update Indicators Visibility for Opposing Teams'
  (Fortnite|FortShippingLog 0 "RED DIAMOND: Going into Update Indicators Visibility for Opposing Teams function.")
  (if (not (Utilities|Time|IsValidTimerHandle (Variables|Default|GetUpdateIndicatorsTimerHandle)))
    (bind _returnvalue (Utilities|Time|SetTimerforNextTickbyFunctionName 0 "Update Indicators Visibility for Opposing Teams"))
    (Variables|Default|SetUpdateIndicatorsTimerHandle _returnvalue)))

(event EventTick
  (params (DeltaSeconds :type Float%20%28single-precision%29)))

(event Custom|SetupCuratedSetlistInfo
  ;[0.6123 1 0.9601] Get the curated setlist value and listen for changes
  (Networking|SwitchHasAuthority
    (:Authority
      (Utilities|FlowControl|Sequence
        (:then_0
          (Utilities|IsValid (Actor|GetComponentbyClass (Game|GetGameState) "/Script/PilgrimQuickplayRuntime.PilgrimSetlistComponent")
            (:Is%20Not%20Valid
              (Fortnite|FortShippingLog 0 "Invalid PilgrimSetlistComponent in SetupCuratedSetlistInfo"))))
        (:then_1
          (bind (_asynctaskproxy _payload) (GameplayEventRouter|StartListeningToEvent (Variables|Default|GetEventRouter) "/Script/PilgrimQuickplayRuntime.PilgrimEvent_CuratedSetlistModeChanged")
            (:OnEventReceived
              (CallFunction|CuratedSetlistStatusChangedEvent))))
        (:then_2
          (CallFunction|CuratedSetlistStatusChangedEvent) ; Process the initial value
        )))))

(event Custom|CuratedSetlistStatusChanged_Event
  (CallFunction|UpdateBasedonGameStateandIsCuratedSetlist))

(event Custom|TryPlayerStateRepAgain
  ;[0.1356 0.4072 0.3763] Used to try again if we get a replicated player state before our widgets are ready
  (Utilities|FlowControl|Delay 0.5)
  (Default|TryUpdatefromPlayerState))

(event Custom|GotoGameplay
  ;[0.1356 0.4072 0.3763] Send the message to advance to gameplay - do some cleanup as we leave
  (Utilities|FlowControl|Sequence
    (:then_0
      (if (not (Variables|Default|GetSuppressGameplayMessages))
        (bind _returnvalue (GameplayEventRouter|GetEventRouter "/Script/FortniteGame.FortPlayspace" (Components|GetOwner)))
        (Utilities|IsValid _returnvalue
          (:Is%20Valid
            (GameplayEventRouter|BroadcastEvent _returnvalue (Utilities|Struct|MakePilgrimQuickplayAllPlayersReadyEvent) 0)))))
    (:then_1
      (Variables|Default|SetAllPlayersReady_Replicated false))))

(event Custom|OnAllPlayersReady
  ;[0.3281 1 0.9202] When all players are ready, check if a tiebreaker is needed. If so, start that process and a timer to begin the exit countdown; otherwise, begin the exit countdown now.
  (Default|SetReturntoResultsInputEnabledforAll false)
  (Default|SetToggleReadyUpInputEnabledforAll false)
  (bind _tiebreaker (Default|CheckforSongTiebreaker))
  (if _tiebreaker
    (Default|PickWinner)
    (Default|SetPilgrimEventTimerState 0 (Utilities|Struct|MakePilgrimEventGreenRoomTimerStateChanged "NewEnumerator3" (Variables|Default|GetTimerLengthWhenPlayerUnfinished)))
    (bind _returnvalue (Utilities|Time|SetTimerbyEvent (EventDispatchers|CreateEvent :!event Custom|DelayedCountdownToExit) (Variables|Default|GetWinnerSequenceDuration)))
    (Variables|Default|SetDelayedCoutdownToExitTimerHandle _returnvalue)
    (else
      (Default|SetPilgrimEventTimerState 0 (Utilities|Struct|MakePilgrimEventGreenRoomTimerStateChanged "NewEnumerator2" (Variables|Default|GetTimerLengthWhenPlayerUnfinished))))))

(event Custom|DelayedCountdownToExit
  ;[0.3281 1 0.9202] Event for timer to call
  (Default|SetPilgrimEventTimerState 0 (Utilities|Struct|MakePilgrimEventGreenRoomTimerStateChanged "NewEnumerator2" (Variables|Default|GetTimerLengthAllReady))))

(event Custom|TickTimer
  ;[0.2658 0.5759 0.7969] Countdown and exit
  (Networking|CanExecuteCosmeticEvents
    (:True
      (CallFunction|SetTimerProgressfromTimer)
      (bind _timerimage (Default|GetTimerImage))
      (Utilities|IsValid _timerimage
        (:Is%20Valid
          (bind _returnvalue (Appearance|GetDynamicMaterial _timerimage :!class /Script/UMG.Image))
          (Rendering|Material|SetScalarParameterValue _returnvalue "Progress" (Variables|Default|GetTimerProgress) :!class /Script/Engine.MaterialInstanceDynamic))))))

(event Custom|CountdowntoExit
  (params (TimeToExit :type Float%20%28double-precision%29))
  (bind _timerlengthallready (Variables|Default|GetTimerLengthAllReady))
  (bind _outvalue (Default|GetDurationWithFFCheatApplied _timerlengthallready))
  (bind _timerlengthpingcountdown (Variables|Default|GetTimerLengthPingCountdown))
  (bind _outvalue_2 (Default|GetDurationWithFFCheatApplied _timerlengthpingcountdown))
  ;[0.2658 0.5759 0.7969] Countdown and exit
  (Networking|CanExecuteCosmeticEvents
    (:True
      (Class|WBPPilgrimQuickplayGreenroomStageTimer|ShowTimer (Variables|/PilgrimQuickplay/Widgets/WBP_GreenRoomDisplayBoard.WBP_GreenRoomDisplayBoard_C|GetStageTimer (CallFunction|GetDisplayBoardWidget))))
    (:False
      _))
  (Components|Tick|SetComponentTickEnabled true)
  (if (Utilities|Time|IsValidTimerHandle (Variables|Default|GetExitGreenroomTimerHandle))
    ;;[1 1 1] If the incoming timer is less than our current time remaining - restart the timer
    (if (> (Utilities|Time|GetTimerRemainingTimebyHandle (Variables|Default|GetExitGreenroomTimerHandle)) TimeToExit)
      _)
    (else
      _))
  (bind _returnvalue (Utilities|Time|SetTimerbyEvent (delegate Custom|TimeUp%20Start%20Game) TimeToExit))
  (Variables|Default|SetExitGreenroomTimerHandle _returnvalue)
  ;;[1 1 1] If our timer is longer than "All Players Ready" timer length, force "All Ready" when we hit that time.
  (if (> TimeToExit _outvalue)
    (bind _returnvalue_2 (Utilities|Time|SetTimerbyEvent (delegate Custom|All%20Ready%20Timer%20Length%20Remaining) (- TimeToExit _outvalue)))
    (Variables|Default|SetForceAllPlayersReadyTimerHandle _returnvalue_2))
  (Development|PrintString (Utilities|String|Append "Starting timer for " (Utilities|String|ToString(Float) _outvalue_2)) true true "(R=0.000000,G=0.660000,B=1.000000,A=1.000000)" 5e+02)
  (bind _returnvalue_3 (Utilities|Time|SetTimerbyEvent (delegate Custom|PingTimeCountdown) _outvalue_2 true))
  (Variables|Default|SetPingTimerHandle _returnvalue_3)
  (Networking|CanExecuteCosmeticEvents
    (:True
      (bind _timerimage (Default|GetTimerImage))
      (Utilities|IsValid _timerimage
        (:Is%20Valid
          (bind _returnvalue_4 (Appearance|GetDynamicMaterial _timerimage :!class /Script/UMG.Image))
          (Rendering|Material|SetScalarParameterValue _returnvalue_4 "Progress" 1.0 :!class /Script/Engine.MaterialInstanceDynamic))
        (:Is%20Not%20Valid
          _))
      (Default|SetPulseTimerInteger (Math|Float|Truncate (- (* TimeToExit (Default|GetTimerCoefWithFFCheatApplied)) (Variables|Default|GetTimerLengthAllReady)))))))

(event Custom|All%20Ready%20Timer%20Length%20Remaining
  ;[0.2658 0.5759 0.7969] Countdown and exit
  ;;[1 1 1] If our timer is longer than "All Players Ready" timer length, force "All Ready" when we hit that time.
  (Default|HandleReadyTimerExpiration))

(event Custom|TimeUp%20Start%20Game
  ;[0.2658 0.5759 0.7969] Countdown and exit
  (Default|SetReturntoResultsInputEnabledforAll false)
  (Default|SetToggleReadyUpInputEnabledforAll false)
  (Variables|Default|SetTimerProgress 1.0)
  (CallFunction|GotoGameplay)
  (Components|Tick|SetComponentTickEnabled false))

(event Custom|PingTimeCountdown
  ;[0.2658 0.5759 0.7969] Countdown and exit
  ;;[1 1 1] Ping timer once per second to show countdown
  (if (Utilities|Time|IsTimerActivebyHandle (Variables|Default|GetExitGreenroomTimerHandle))
    (CallFunction|SetTimerProgressfromTimer)
    (Networking|CanExecuteCosmeticEvents
      (:True
        (Default|SetPulseTimerInteger (Math|Conversions|ToInteger(Integer64) (Math|Integer64|Absolute(Integer64) (Math|Conversions|ToInteger64(Integer) (Math|Float|Round (- (* (Utilities|Time|GetTimerRemainingTimebyHandle (Variables|Default|GetExitGreenroomTimerHandle)) (Default|GetTimerCoefWithFFCheatApplied)) (Variables|Default|GetTimerLengthAllReady)))))))))
    (else
      (Utilities|Time|ClearandInvalidateTimerbyHandle (Variables|Default|GetPingTimerHandle))
      (Default|SetPulseTimerInteger 0))))

(event Custom|SetTimerProgressFromTimer
  (bind _exitgreenroomtimerhandle (Variables|Default|GetExitGreenroomTimerHandle))
  (bind _returnvalue (Utilities|Time|GetTimerElapsedTimebyHandle _exitgreenroomtimerhandle))
  (bind _exitgreenroomtimerhandle_2 (Variables|Default|GetExitGreenroomTimerHandle))
  (bind _returnvalue_2 (Utilities|Time|GetTimerRemainingTimebyHandle _exitgreenroomtimerhandle_2))
  (bind _outvalue (Default|GetTimerCoefWithFFCheatApplied))
  (bind _returnvalue_3 (* _returnvalue _outvalue))
  (bind _returnvalue_4 (* _returnvalue_2 _outvalue))
  (bind _returnvalue_5 (+ _returnvalue_4 _returnvalue_3))
  ;[1 1 1] Set the 'TimerProgress' variable from our current timer.
  (if (Utilities|Time|IsValidTimerHandle (Variables|Default|GetExitGreenroomTimerHandle))
    (if (> _returnvalue_5 0)
      (Variables|Default|SetTimerProgress (- 1.0 (/ _returnvalue_4 _returnvalue_5))))))

(event Custom|StopAndHideTimer
  (bind _wbpdisplayboard (CallFunction|GetDisplayBoardWidget))
  (bind _stagetimer (Variables|/PilgrimQuickplay/Widgets/WBP_GreenRoomDisplayBoard.WBP_GreenRoomDisplayBoard_C|GetStageTimer _wbpdisplayboard))
  ;[1 1 1] Stop and Hide Timer
  (Utilities|Time|ClearandInvalidateTimerbyHandle (Variables|Default|GetExitGreenroomTimerHandle))
  (Utilities|Time|ClearandInvalidateTimerbyHandle (Variables|Default|GetForceAllPlayersReadyTimerHandle))
  (Utilities|Time|ClearandInvalidateTimerbyHandle (Variables|Default|GetPingTimerHandle))
  (Components|Tick|SetComponentTickEnabled false)
  (Networking|CanExecuteCosmeticEvents
    (:True
      (Utilities|IsValid _wbpdisplayboard
        (:Is%20Valid
          (Utilities|IsValid _stagetimer
            (:Is%20Valid
              (Class|WBPPilgrimQuickplayGreenroomStageTimer|HideTimer _stagetimer))))))))

(event Custom|Update%20Tick%20Timer
  (params (DeltaSeconds :type Float%20%28single-precision%29))
  ;[1 1 1] Tick (Enabled only when timer is counting down)
  (CallFunction|TickTimer))

(event Custom|SetupMyPlayerStates
  ;[1 1 1] Setup MyPlayer States
  (Utilities|IsValid (Variables|Default|GetGreenRoomPlayerDisplay)
    (:Is%20Valid
      (Default|TryUpdatefromPlayerState)
      (Default|CallPlayerStatesChanged))
    (:Is%20Not%20Valid
      (Utilities|FlowControl|DelayUntilNextTick)
      (CallFunction|SetupMyPlayerStates))))

(event Custom|NetSyncPawnRotation
  (params (Pawn :type Fort%20Player%20Pawn%20Object%20Reference) (InRot :type Rotator)))

(event Custom|OnPilgrimTimerStateReplicated
  (bind _bp_pilgrim_quickplay_playspace (Variables|Default|GetBP%20Pilgrim%20Quickplay%20Playspace))
  (bind _returnvalue (Class|PilgrimQuickplayBandStateComponent|GetBandState _bp_pilgrim_quickplay_playspace))
  (bind _pilgrimtimerstate_replicated_greenroomtimerstate_2_04fc39c74228c50f7ae0f7b72d379be4 (Utilities|Struct|BreakPilgrimEventGreenRoomTimerStateChanged (Variables|GetPilgrimTimerState_Replicated_GreenRoomTimerState_2_04FC39C74228C50F7AE0F7B72D379BE4)))
  (bind (_pilgrimtimerstate_replicated_greenroomtimerstate_2_04fc39c74228c50f7ae0f7b72d379be4_2 _pilgrimtimerstate_replicated_finaltimerlength_5_73a42f394ba6b986d4f14ba69df7bff0) (Utilities|Struct|BreakPilgrimEventGreenRoomTimerStateChanged (Variables|GetPilgrimTimerState_Replicated_GreenRoomTimerState_2_04FC39C74228C50F7AE0F7B72D379BE4)))
  (if (not (Variables|Default|GetSuppressGameplayMessages))
    (Utilities|FlowControl|Sequence
      (:then_0
        (Variables|Default|GetBP%20Pilgrim%20Quickplay%20Playspace
          (:then
            _)
          (:else
            (bind _asbp_pilgrim_quickplay_playspace (Utilities|Casting|CastToBP_PilgrimQuickplayPlayspace (Components|GetOwner))
              (:then
                (Variables|Default|SetBP%20Pilgrim%20Quickplay%20Playspace _asbp_pilgrim_quickplay_playspace))
              (:CastFailed
                (Development|PrintString "BP_PilgrimQuickplay_PlayerManager_PlayspaceComponent's parent failed cast to BP_PilgrimQuickplayPlayspace! [OnPilgrimTimerStateReplicated]" true true "(R=0.906250,G=0.195980,B=0.250125,A=1.000000)" 2e+01)))))
        (Variables|Default|GetEventRouter
          (:then
            _)
          (:else
            ;[0.01983 0.15 0.04994 0.5] [FORT-962847] Sometimes EventRouter is not yet available once the TimerState was replicated, here we're making sure to run the Client-side logic for updating the UI, as soon as EventRouter becomes available
            (bind _eventrouter (Pilgrim|PilgrimWaitForEventRouter (Variables|Default|GetBP%20Pilgrim%20Quickplay%20Playspace))
              (:OnEventRouterAvailable
                (Variables|Default|SetEventRouter _eventrouter)))))
        (GameplayEventRouter|BroadcastEvent (Variables|Default|GetEventRouter) (Variables|Default|GetPilgrimTimerState_Replicated) 0)
        (switch Utilities|FlowControl|Switch|SwitchonGreenRoomTimerState _pilgrimtimerstate_replicated_greenroomtimerstate_2_04fc39c74228c50f7ae0f7b72d379be4_2
          (:NewEnumerator0
            _)
          (:NewEnumerator1
            (CallFunction|CountdowntoExit (Default|GetDurationWithFFCheatApplied (+ _pilgrimtimerstate_replicated_finaltimerlength_5_73a42f394ba6b986d4f14ba69df7bff0 (Variables|Default|GetTimerLengthAllReady)))))
          (:NewEnumerator2
            (CallFunction|CountdowntoExit (Default|GetDurationWithFFCheatApplied (Variables|Default|GetTimerLengthAllReady))))
          (:NewEnumerator3
            _))
        (CallFunction|StopandHideTimer))
      (:then_1
        (Utilities|IsValid _returnvalue
          (:Is%20Valid
            (for _array_element (Class|PilgrimQuickplayBandStateComponent|GetPlayersinBand _returnvalue (Utilities|Enum|LiteralenumEFortTeam "HumanPvP_Team1"))
              (Utilities|IsValid (PlayerState|GetPlayerController _array_element)
                (:Is%20Valid
                  (Utilities|IsValid (Actor|GetComponentbyClass (PlayerState|GetPlayerController _array_element) "/PilgrimCore/Player/Components/BP_PilgrimPlayerControllerComponent.BP_PilgrimPlayerControllerComponent_C")
                    (:Is%20Valid
                      (Class|BPPilgrimPlayerControllerComponent|ToggleGreenroomMappingContext (Actor|GetComponentbyClass (PlayerState|GetPlayerController _array_element) "/PilgrimCore/Player/Components/BP_PilgrimPlayerControllerComponent.BP_PilgrimPlayerControllerComponent_C") (and (Utilities|Enum|NotEqual(Enum) _pilgrimtimerstate_replicated_greenroomtimerstate_2_04fc39c74228c50f7ae0f7b72d379be4 "NewEnumerator2") (Utilities|Enum|Equal(Enum) (Class|PilgrimQuickplayStateMachine|GetActivePilgrimStateId (Class|PilgrimQuickplayPlayspace|GetPilgrimQuickplayStateMachine (Variables|Default|GetBP%20Pilgrim%20Quickplay%20Playspace))) "Pregame"))))))))))))))

(event Custom|SetupClientListeningEvent
  (Utilities|FlowControl|Sequence
    (:then_0
      ;[1 0.8518 0.662] Player instrument changed events
      (bind (_asynctaskproxy _payload) (GameplayEventRouter|StartListeningToEvent (Variables|Default|GetEventRouter) "/Script/PilgrimCoreRuntime.PilgrimEvent_PlayerState_InstrumentChanged")
        (:OnEventReceived
          (bind (_payload_playerstate _payload_newinstrumentchoice) (Utilities|Struct|BreakPilgrimEventPlayerStateInstrumentChanged _payload))
          (bind (_matchfound _matchingindex) (Default|FindMatchingPlayer _payload_playerstate))
          (if _matchfound
            (Class|WBPGreenRoomPlayerDisplay|SetInstrumentType (Utilities|Array|Get(acopy) (CallFunction|GetGreenRoomPlayerDisplayWidgets) _matchingindex) _payload_newinstrumentchoice)))))
    (:then_1
      ;[1 0.8518 0.662] Player difficulty changed events
      (bind (_asynctaskproxy_2 _payload_2) (GameplayEventRouter|StartListeningToEvent (Variables|Default|GetEventRouter) "/Script/PilgrimCoreRuntime.PilgrimEvent_PlayerState_DifficultyChanged")
        (:OnEventReceived
          (bind (_payload_playerstate_2 _payload_newsongdifficulty) (Utilities|Struct|BreakPilgrimEventPlayerStateDifficultyChanged _payload_2))
          (bind (_matchfound_2 _matchingindex_2) (Default|FindMatchingPlayer _payload_playerstate_2))
          (if _matchfound_2
            (Class|WBPGreenRoomPlayerDisplay|SetDifficultyDisplay (Utilities|Array|Get(acopy) (CallFunction|GetGreenRoomPlayerDisplayWidgets) _matchingindex_2) _payload_newsongdifficulty)))))
    (:then_2
      ;[1 0.8518 0.662] Player song vote changed events - NOTE: This is NOT for the upcoming setlist. This is the player's individual vote event.
      (bind (_asynctaskproxy_3 _payload_3) (GameplayEventRouter|StartListeningToEvent (Variables|Default|GetEventRouter) "/Script/PilgrimCoreRuntime.PilgrimEvent_PlayerState_SongVoteChanged")
        (:OnEventReceived
          (bind (_payload_forplayer _payload_newsongvote) (Utilities|Struct|BreakPilgrimEventPlayerStateSongVoteChanged _payload_3))
          (bind (_matchfound_3 _matchingindex_3) (Default|FindMatchingPlayer (Variables|/Script/Engine.Pawn|GetPlayerState _payload_forplayer)))
          (if _matchfound_3
            (Class|WBPGreenRoomPlayerDisplay|SetPlayerVote (Utilities|Array|Get(acopy) (CallFunction|GetGreenRoomPlayerDisplayWidgets) _matchingindex_3) _payload_newsongvote)))))
    (:then_3
      ;[1 0.8518 0.662] Individual player readiness events
      (bind (_asynctaskproxy_4 _payload_4) (GameplayEventRouter|StartListeningToEvent (Variables|Default|GetEventRouter) "/Script/PilgrimCoreRuntime.PilgrimEvent_PlayerState_ReadinessChanged")
        (:OnEventReceived
          (bind (_payload_forplayer_2 _payload_bisready) (Utilities|Struct|BreakPilgrimEventPlayerStateReadinessChanged _payload_4))
          (bind (_matchfound_4 _matchingindex_4) (Default|FindMatchingPlayer (Variables|/Script/Engine.Pawn|GetPlayerState _payload_forplayer_2)))
          (if _matchfound_4
            (Class|WBPGreenRoomPlayerDisplay|SetReadyStateUI (Utilities|Array|Get(acopy) (CallFunction|GetGreenRoomPlayerDisplayWidgets) _matchingindex_4) (select (Variables|Default|GetAllPlayersReady_Replicated) true _payload_bisready))))))
    (:then_4
      ;[1 0.8518 0.662] Reset board dispaly on the client
      (bind (_asynctaskproxy_5 _payload_5) (GameplayEventRouter|StartListeningToEvent (Variables|Default|GetEventRouter) "/Script/PilgrimQPSharedRuntime.PilgrimEvent_QuickplayState_Began")
        (:OnEventReceived
          (bind _payload_stateid (Utilities|Struct|BreakPilgrimEventQuickplayStateBegan _payload_5))
          (switch Utilities|FlowControl|Switch|SwitchonEPilgrimQuickplayState _payload_stateid
            (:SongGameplay
              (Class|WBPGreenRoomDisplayBoard|ResetBoard (CallFunction|GetDisplayBoardWidget)))))))
    (:then_5
      ;[1 1 1] Listen for ready up input
      (bind (_asynctaskproxy_6 _payload_6) (GameplayEventRouter|StartListeningToEvent (Variables|Default|GetEventRouter) "/PilgrimCore/Player/Components/PilgrimEvent_TogglePlayerReadyButtonPressed.PilgrimEvent_TogglePlayerReadyButtonPressed")
        (:OnEventReceived
          (bind (_payload_fortplayercontroller_11_9833968f4a2479a71e2b2ba53665a4e7 _payload_btriggered_15_1d71664740efa29164a7fbab2be0db9b) (Utilities|Struct|BreakPilgrimEventTogglePlayerReadyButtonPressed _payload_6))
          (if _payload_btriggered_15_1d71664740efa29164a7fbab2be0db9b
            (if (> (Class|PilgrimSetlistInterface|GetNumSongs (Variables|/PilgrimQuickplay/Playspaces/BP_PilgrimQuickplayPlayspace.BP_PilgrimQuickplayPlayspace_C|GetPilgrimCustomSetlist (Class|SparksMusicPlayspace|GetSparksMusicPlayspaceOfClass "/PilgrimQuickplay/Playspaces/BP_PilgrimQuickplayPlayspace.BP_PilgrimQuickplayPlayspace_C" _payload_fortplayercontroller_11_9833968f4a2479a71e2b2ba53665a4e7))) 0)
              (if (Class|BPPilgrimPlayerControllerComponent|CheckforAcceptingInput (Actor|GetComponentbyClass _payload_fortplayercontroller_11_9833968f4a2479a71e2b2ba53665a4e7 "/PilgrimCore/Player/Components/BP_PilgrimPlayerControllerComponent.BP_PilgrimPlayerControllerComponent_C") "IA_PilgrimToggleReady" "(TagName=\"Pilgrim.GreenRoom.Input.ToggleReadyUp\")" true)
                (Class|BPPilgrimPlayerControllerComponent|ToggleReadyInputPressed (Actor|GetComponentbyClass _payload_fortplayercontroller_11_9833968f4a2479a71e2b2ba53665a4e7 "/PilgrimCore/Player/Components/BP_PilgrimPlayerControllerComponent.BP_PilgrimPlayerControllerComponent_C"))
                (else
                  (Fortnite|FortShippingLog 0 "Not accepting ready up because we don't have the associated gameplay tag")))
              (else
                _))
            (else
              (if (<= (Class|PilgrimSetlistInterface|GetNumSongs (Variables|/PilgrimQuickplay/Playspaces/BP_PilgrimQuickplayPlayspace.BP_PilgrimQuickplayPlayspace_C|GetPilgrimCustomSetlist (Class|SparksMusicPlayspace|GetSparksMusicPlayspaceOfClass "/PilgrimQuickplay/Playspaces/BP_PilgrimQuickplayPlayspace.BP_PilgrimQuickplayPlayspace_C" _payload_fortplayercontroller_11_9833968f4a2479a71e2b2ba53665a4e7))) 0)
                _)))
          (bind _asfort_player_controller (Utilities|Casting|CastToFortPlayerController (Components|GetOwner)))
          (GameplayEventRouter|BroadcastEvent (Variables|Default|GetEventRouter) (Utilities|Struct|MakePilgrimEventReadyUpPressedwhenDisabled _asfort_player_controller) 0))))))

(event ApplicationLifecycle|EventOnAppResumed
  (params (ElapsedTimeSeconds :type Float%20%28single-precision%29))
  (bind _exitgreenroomtimerhandle (Variables|Default|GetExitGreenroomTimerHandle))
  (if (Utilities|Time|IsValidTimerHandle _exitgreenroomtimerhandle)
    (bind _returnvalue (Utilities|Time|SetTimerbyEvent (EventDispatchers|CreateEvent :!event TimeUp%20Start%20Game) (Math|Float|Max(Float) (- (Utilities|Time|GetTimerRemainingTimebyHandle _exitgreenroomtimerhandle) ElapsedTimeSeconds) 0.01)))
    (Utilities|FlowControl|Sequence
      (:then_0
        (Utilities|Time|ClearandInvalidateTimerbyHandle (Variables|Default|GetExitGreenroomTimerHandle)))
      (:then_1
        (Variables|Default|SetExitGreenroomTimerHandle _returnvalue)))))

(event Custom|SetupServerListeningEvents
  (bind _pilgrimtimerstate_replicated_greenroomtimerstate_2_04fc39c74228c50f7ae0f7b72d379be4 (Utilities|Struct|BreakPilgrimEventGreenRoomTimerStateChanged (Variables|GetPilgrimTimerState_Replicated_GreenRoomTimerState_2_04FC39C74228C50F7AE0F7B72D379BE4)))
  (Utilities|FlowControl|Sequence
    (:then_0
      ;[0.07146 1 0.3317] SERVER - Listen for all players ready event and set replicated value
      (bind (_asynctaskproxy _payload) (GameplayEventRouter|StartListeningToEvent (Variables|Default|GetEventRouter) "/Script/PilgrimCoreRuntime.PilgrimEvent_BandState_ReadinessChanged")
        (:OnEventReceived
          (Default|CheckForAllPlayersReady (Utilities|Struct|BreakPilgrimEventBandStateReadinessChanged _payload)))))
    (:then_1
      ;[0.1506 1 0.6729] SERVER - Setlist voting/results
      (bind (_asynctaskproxy_2 _payload_2) (GameplayEventRouter|StartListeningToEvent (Variables|Default|GetEventRouter) "/Script/PilgrimCoreRuntime.PilgrimEvent_BandState_SetlistVoteWinnerChanged")
        (:OnEventReceived
          (bind _payload_newvotewinner (Utilities|Struct|BreakPilgrimEventBandStateSetlistVoteWinnerChanged _payload_2))
          (Variables|Default|SetVoteWinner_Server _payload_newvotewinner)
          (Default|CollectDataforReplication))))
    (:then_2
      (bind (_asynctaskproxy_3 _payload_3) (GameplayEventRouter|StartListeningToEvent (Variables|Default|GetEventRouter) "/Script/PilgrimCoreRuntime.PilgrimEvent_PlayerState_SetlistVoteChanged")
        (:OnEventReceived
          (bind (_payload_forplayer _payload_newsetlistvote _payload_prevsetlistvote) (Utilities|Struct|BreakPilgrimEventPlayerStateSetlistVoteChanged _payload_3))
          (Default|RecordPlayerVote _payload_newsetlistvote (Variables|/Script/Engine.Pawn|GetPlayerState _payload_forplayer))
          (Default|HandlePreviousPlayerVote _payload_prevsetlistvote)
          (Default|HandleNewPlayerVote _payload_newsetlistvote)
          (Default|CollectDataforReplication))))
    (:then_3
      ;[0.5473 1 0.8263] SERVER - Listen For Pilgrim Game State begin -- For Outro, clear any votes when the player enters gameplay so all is fresh when they return to the green room - clear the winner
      (bind (_asynctaskproxy_4 _payload_4) (GameplayEventRouter|StartListeningToEvent (Variables|Default|GetEventRouter) "/Script/PilgrimQPSharedRuntime.PilgrimEvent_QuickplayState_Began")
        (:OnEventReceived
          (Variables|Default|SetCurrent%20Pilgrim%20State (Utilities|Struct|BreakPilgrimEventQuickplayStateBase _payload_4)) ; This value will be overwritten in the next event\nunless something goes wrong. It is set here as a\nfallback
          (CallFunction|UpdateBasedonGameStateandIsCuratedSetlist))))
    (:then_4
      ;[0.5473 1 0.8263] Listen for Pilgrim Game State Ending
      (bind (_asynctaskproxy_5 _payload_5) (GameplayEventRouter|StartListeningToEvent (Variables|Default|GetEventRouter) "/Script/PilgrimQPSharedRuntime.PilgrimEvent_QuickplayState_Ended")
        (:OnEventReceived
          (bind _payload_stateid (Utilities|Struct|BreakPilgrimEventQuickplayStateEnded _payload_5))
          (switch Utilities|FlowControl|Switch|SwitchonEPilgrimQuickplayState _payload_stateid
            (:Pregame
              (Default|SetSwapDifficultyInputEnabledforAll false)
              (Default|SetToggleReadyUpInputEnabledforAll false))
            (:Intro
              (Default|UpdateIndicatorsVisibilityforOpposingTeams))))))
    (:then_5
      ;[1 1 1] Listen for individual players ready and update/reset timer as needed.
      (bind (_asynctaskproxy_6 _payload_6) (GameplayEventRouter|StartListeningToEvent (Variables|Default|GetEventRouter) "/Script/PilgrimCoreRuntime.PilgrimEvent_PlayerState_ReadinessChanged")
        (:OnEventReceived
          (Utilities|FlowControl|Sequence
            (:then_0
              (Variables|Default|SetNumReadyPlayers 0)
              (for _array_element (Variables|Default|GetAllPlayerStates)
                (Utilities|IsValid _array_element
                  (:Is%20Valid
                    (bind _returnvalue (Class|PilgrimQuickplayPlayerStateComponent|GetPilgrimQuickplayPlayerStateComponent _array_element))
                    (if (Class|PilgrimCorePlayerStateComponent|GetIsReady _returnvalue)
                      (Math|Integer|IncrementInt (Variables|Default|GetNumReadyPlayers))))))
              (Utilities|FlowControl|Sequence
                (:then_0
                  (if (== (Variables|Default|GetNumReadyPlayers) 0)
                    (if (not (Variables|Default|GetisCuratedSetlist))
                      (if (> (Variables|Default|GetPrevNumReadyPlayers) 0)
                        (Default|SetPilgrimEventTimerState 0 (struct :GreenRoomTimerState_2_04FC39C74228C50F7AE0F7B72D379BE4 "NewEnumerator0" :FinalTimerLength_5_73A42F394BA6B986D4F14BA69DF7BFF0 0.0))))
                    (elif (<= (Variables|Default|GetPrevNumReadyPlayers) 0)
                      (if (not (Utilities|Enum|Equal(Enum) _pilgrimtimerstate_replicated_greenroomtimerstate_2_04fc39c74228c50f7ae0f7b72d379be4 "NewEnumerator2"))
                        (Default|SetPilgrimEventTimerState 0 (Utilities|Struct|MakePilgrimEventGreenRoomTimerStateChanged "NewEnumerator1" (Variables|Default|GetTimerLengthWhenPlayerUnfinished)))))))
                (:then_1
                  (Variables|Default|SetPrevNumReadyPlayers (Variables|Default|GetNumReadyPlayers)))))
            (:then_1
              (bind (_payload_forplayer_2 _payload_bisready) (Utilities|Struct|BreakPilgrimEventPlayerStateReadinessChanged _payload_6))
              (if (not _payload_bisready)
                (Variables|Default|SetAllPlayersReady_Replicated false)))))))
    (:then_6
      ;[0.1356 0.4072 0.3763] Listen for quickplay begin states on the server - if we go into a pregame state, rebuild the player list and big board display
      (bind (_asynctaskproxy_7 _payload_7) (GameplayEventRouter|StartListeningToEvent (Variables|Default|GetEventRouter) "/Script/PilgrimQPSharedRuntime.PilgrimEvent_QuickplayState_BeganServer")
        (:OnEventReceived
          (if (Utilities|Enum|Equal(Enum) (Utilities|Struct|BreakPilgrimEventQuickplayStateBase _payload_7) "Pregame")
            (bind _aspilgrim_quickplay_playspace (Utilities|Casting|CastToPilgrimQuickplayPlayspace (Components|GetOwner)))
            ;;[0.9161 1 0.156] update base info from the current set of band members (folks may have dropped during gameplay, and we don't want to remember them anymore)
            (if (Utilities|Array|IsNotEmpty (Class|PilgrimQuickplayBandStateComponent|GetPlayersinBand (Class|PilgrimQuickplayPlayspace|GetBandState _aspilgrim_quickplay_playspace) 0))
              (Variables|Default|SetMyPlayerStates (Class|PilgrimQuickplayBandStateComponent|GetPlayersinBand (Class|PilgrimQuickplayPlayspace|GetBandState _aspilgrim_quickplay_playspace) 0)))))))
    (:then_7
      ;[1 1 1] Listen for setlist changed
      (bind (_asynctaskproxy_8 _payload_8) (GameplayEventRouter|StartListeningToEvent (Variables|Default|GetEventRouter) "/Script/PilgrimCoreRuntime.PilgrimEvent_BandState_CustomSetlistChanged")
        (:OnEventReceived
          (bind _payload_changedsetlist (Utilities|Struct|BreakPilgrimEventBandStateCustomSetlistChanged _payload_8))
          (if (<= (Class|PilgrimSetlistInterface|GetNumSongs _payload_changedsetlist) 0)
            (for _array_element_2 (Variables|Default|GetMyPlayerStates)
              (bind _returnvalue_2 (Class|PilgrimQuickplayPlayerStateComponent|GetPilgrimQuickplayPlayerStateComponent _array_element_2))
              (Utilities|IsValid _returnvalue_2
                (:Is%20Valid
                  (if (Class|PilgrimCorePlayerStateComponent|GetIsReady _returnvalue_2)
                    (Class|PilgrimCorePlayerStateComponent|ServerSetIsReady _returnvalue_2 false)))))))))))

(event Custom|UpdateBasedOnGameStateAndIsCuratedSetlist
  (bind _returnvalue (Game|GetGameState))
  (bind _returnvalue_2 (Actor|GetComponentbyClass _returnvalue "/Script/PilgrimQuickplayRuntime.PilgrimSetlistComponent"))
  ;[0.5473 1 0.8263] SERVER - Listen For Pilgrim Game State begin -- For Outro, clear any votes when the player enters gameplay so all is fresh when they return to the green room - clear the winner
  (Utilities|IsValid _returnvalue_2
    (:Is%20Valid
      (Variables|Default|SetisCuratedSetlist (Class|PilgrimSetlistComponent|IsCuratedSetlistMode _returnvalue_2)))
    (:Is%20Not%20Valid
      (Fortnite|FortShippingLog 0 "Invalid PilgrimSetlistComponent in UpdateBasedOnGameStateAndIsCurratedSetlist")))
  (bind _asbp_pilgrim_quickplay_playspace (Utilities|Casting|CastToBP_PilgrimQuickplayPlayspace (Components|GetOwner))
    (:then
      (Variables|Default|SetCurrent%20Pilgrim%20State (Class|PilgrimQuickplayStateMachine|GetActivePilgrimStateId (Class|PilgrimQuickplayPlayspace|GetPilgrimQuickplayStateMachine _asbp_pilgrim_quickplay_playspace))))
    (:CastFailed
      (Fortnite|FortShippingLog 0 "Invalid BP_PilgrimQuickplayPlayspace in UpdateBasedOnGameStateAndIsCurratedSetlist")))
  (switch Utilities|FlowControl|Switch|SwitchonEPilgrimQuickplayState (Variables|Default|GetCurrent%20Pilgrim%20State)
    (:Pregame
      (Default|ResetReadyStateforAllPlayers)
      (Variables|SetPilgrimTimerState_Replicated_GreenRoomTimerState_2_04FC39C74228C50F7AE0F7B72D379BE4 (Utilities|Struct|MakePilgrimEventGreenRoomTimerStateChanged (select (Variables|Default|GetisCuratedSetlist) "NewEnumerator1" "NewEnumerator0") (Math|Float|SelectFloat (Variables|Default|GetTimerLengthWhenPlayerUnfinished_CuratedSetlist) (Variables|Default|GetTimerLengthWhenPlayerUnfinished) (Variables|Default|GetisCuratedSetlist))))
      (Variables|Default|SetAllPlayersReady_Replicated false)
      (Default|SetToggleReadyUpInputEnabledforAll true)
      (Default|SetSwapDifficultyInputEnabledforAll true)
      (if (Default|IsLastorSingleSong)
        (Default|VoteNothingforAllVoters)
        (Default|UpdateIndicatorsVisibilityforOpposingTeams)))
    (:Outro
      (if (Default|IsLastorSingleSong)
        (Variables|Default|SetVoteWinner_Server 0)
        (Default|CollectDataforReplication)))
    (:SongResults
      _)
    (:SetResults
      _))
  (Default|ResetReadyStateforAllPlayers)
  (Variables|Default|SetAllPlayersReady_Replicated false)
  (Variables|SetPilgrimTimerState_Replicated_GreenRoomTimerState_2_04FC39C74228C50F7AE0F7B72D379BE4 (struct :GreenRoomTimerState_2_04FC39C74228C50F7AE0F7B72D379BE4 "NewEnumerator0" :FinalTimerLength_5_73A42F394BA6B986D4F14BA69DF7BFF0 0.0))
  (Default|SetToggleReadyUpInputEnabledforAll true)
  (Default|SetSwapDifficultyInputEnabledforAll true)
  (Default|SetReturntoResultsInputEnabledforAll true))

(fn CheckforAllPlayersReady
  (params (InAllPlayersKnownReady :type Boolean))
  (Development|PrintString (Utilities|String|Append "Check for all players ready " (Utilities|String|ToString(Boolean) InAllPlayersKnownReady)) true true "(R=1.000000,G=0.062211,B=0.702196,A=1.000000)" 12.0)
  (if (not (Variables|Default|GetAllPlayersReady_Replicated))
    (LocalVariables|SetLocal_AllPlayersReady true)
    ;[1 0.1415 0.9196] We only check for ready players if we are not _told_ all player are ready.
    (if (not InAllPlayersKnownReady)
      (bind _array_element (Utilities|Array|ForEachLoopwithBreak (Variables|Default|GetAllPlayerStates))
        (:LoopBody
          (Utilities|IsValid _array_element
            (:Is%20Valid
              (bind _asfort_player_state_athena (Utilities|Casting|CastToFortPlayerStateAthena _array_element)
                (:then
                  (if (not (State|IsDisconnected _asfort_player_state_athena))
                    (bind _returnvalue (Class|PilgrimQuickplayPlayerStateComponent|GetPilgrimQuickplayPlayerStateComponent _array_element))
                    (Utilities|IsValid _returnvalue
                      (:Is%20Valid
                        (if (not (Class|PilgrimCorePlayerStateComponent|GetIsReady _returnvalue))
                          (LocalVariables|SetLocal_AllPlayersReady false))
                        (Math|Integer|IncrementInt (Variables|Default|GetNumReadyPlayers)))
                      (:Is%20Not%20Valid
                        _))))
                (:CastFailed
                  _)) ; Filter out any unexpected player states - casting to the expected type
            )
            (:Is%20Not%20Valid
              _)))
        (:Completed
          _)))
    (if (LocalVariables|GetLocal_AllPlayersReady)
      (Variables|Default|SetAllPlayersReady_Replicated (Variables|GetInAllPlayersKnownReady))
      (Default|ForceAllPlayersReady))))

(fn CheckForSongTiebreaker
  (bind (_votewinner_12_7c6c203e48fb20e81513929552288680 _setliststoshow_21_c2d8ba5f4ac1ba09c749a48a73fcb4df) (Utilities|Struct|BreakSongDisplayDataStruct (Variables|Default|GetSongDisplayData)))
  ;[1 1 1] We have a tiebreaker if we have more than one song to show
  (return (and (Utilities|Enum|Equal(Enum) (Variables|Default|GetCurrent%20Pilgrim%20State) "Pregame") (> (Utilities|Array|Length _setliststoshow_21_c2d8ba5f4ac1ba09c749a48a73fcb4df) 1))))

(fn CollectDataForReplication
  (bind _votewinner_server (Variables|Default|GetVoteWinner_Server))
  (bind _votedforsetlist_server (Variables|Default|GetVotedForSetlist_Server))
  (bind (_value _returnvalue) (Utilities|Map|Find _votedforsetlist_server _votewinner_server))
  (if _returnvalue
    ;[0.4766 0.9502 1] Make a list of all the setlists to show on the big display board
    (Utilities|Array|Clear (Variables|Default|GetSetlistToShow_Server))
    (bind _keys (Utilities|Map|Keys (Variables|Default|GetVotedForSetlist_Server)))
    (for _array_element _keys
      (if (== (Utilities|Map|Find (Variables|Default|GetVotedForSetlist_Server) _array_element) _value)
        (Utilities|Array|Add (Variables|Default|GetSetlistToShow_Server) _array_element)))
    ;[0.4766 0.9502 1] Split the Player Votes Map into two aligned arrays we can replicate and recombine on the client since we can't replicate maps
    (bind _keys_2 (Utilities|Map|Keys (Variables|Default|GetPlayerVotes_Server)))
    (bind _output_get (LocalVariables|SetPlayerKeys _keys_2))
    (for _array_element_2 _output_get
      (Utilities|Array|Add (LocalVariables|GetPlayerVotes) (Utilities|Map|Find (Variables|Default|GetPlayerVotes_Server) _array_element_2)))
    ;[0.4766 0.9502 1] Set and replicate the collected data, packaged in a struct
    (Variables|Default|SetSongDisplayData (Utilities|Struct|MakeSongDisplayDataStruct (Variables|Default|GetVoteWinner_Server) (Variables|Default|GetSetlistToShow_Server) (LocalVariables|GetPlayerKeys) (LocalVariables|GetPlayerVotes)))))

(fn FindMatchingPlayer
  (params (PlayerState :type Player%20State%20Object%20Reference))
  (bind _array_element (Utilities|Array|ForEachLoopwithBreak (CallFunction|GetGreenRoomPlayerDisplayWidgets))
    (:LoopBody
      (Utilities|IsValid (Variables|/PilgrimQuickplay/Widgets/WBP_GreenRoomPlayerDisplay.WBP_GreenRoomPlayerDisplay_C|GetMyPlayerState _array_element)
        (:Is%20Valid
          (LocalVariables|SetCurrentPlayerState_Local (Variables|/PilgrimQuickplay/Widgets/WBP_GreenRoomPlayerDisplay.WBP_GreenRoomPlayerDisplay_C|GetMyPlayerState _array_element))
          (if (Default|IsPlayeronLocalPlayerTeam (LocalVariables|GetCurrentPlayerState_Local))
            (LocalVariables|SetTeamIndex_Local (+ (LocalVariables|GetTeamIndex_Local) 1))
            (if (== (Variables|GetPlayerState) (LocalVariables|GetCurrentPlayerState_Local))
              (LocalVariables|SetMatchFound_Local true)
              (LocalVariables|SetMatchingIndex_Local (LocalVariables|GetTeamIndex_Local)))))))
    (:Completed
      _))
  (return (LocalVariables|GetMatchFound_Local) (LocalVariables|GetMatchingIndex_Local)))

(fn ForceAllPlayersReady
  (for _array_element (Variables|Default|GetAllPlayerStates)
    (Utilities|IsValid _array_element
      (:Is%20Valid
        (bind _returnvalue (Class|PilgrimQuickplayPlayerStateComponent|GetPilgrimQuickplayPlayerStateComponent _array_element))
        (Class|PilgrimCorePlayerStateComponent|ServerSetIsReady _returnvalue true))))
  (Variables|Default|SetAllPlayersReady_Replicated true))

(fn GA_ScoutDrone_RadiusMark_Func
  (params (IndicatedActor :type Actor%20Object%20Reference)))

(fn GetDisplayedPlayerStates
  (return (Variables|Default|GetMyPlayerStates)))

(fn GetDurationwithFFCheatApplied
  (params (InValue :type Float%20%28double-precision%29))
  (if (Editor|InEditororPIE)
    (if (> (Fortnite|GetNumericCVar "Pilgrim.SkipIntro") 0)
      (return (/ InValue (Default|GetTimerCoefWithFFCheatApplied)))
      (else
        _)))
  (return InValue))

(fn GetGreenRoomPlayerDisplayWidgets
  (bind _greenroomplayerdisplay (Variables|Default|GetGreenRoomPlayerDisplay))
  (Utilities|IsValid _greenroomplayerdisplay
    (:Is%20Valid
      (return (Variables|/PilgrimQuickplay/Actors/GreenRoomPlayerDisplay/BP_GreenRoomPlayerDisplay.BP_GreenRoomPlayerDisplay_C|GetGreenRoomPlayerDisplayWidgets _greenroomplayerdisplay)))
    (:Is%20Not%20Valid
      (Fortnite|FortShippingLog 0 "ERROR: 'Get Green Room Player Display Widgets' called in BP_PilgrimQuickplay_PlayerManager_PlayspaceComponent, but GreenRoomPlayerDisplay is not valid."))))

(fn GetTimerCoefwithFFCheatApplied
  (if (Editor|InEditororPIE)
    (if (> (Fortnite|GetNumericCVar "Pilgrim.SkipIntro") 0)
      (return 1e+01)
      (else
        _)))
  (return 1.0))

(fn GetDisplayBoardWidget
  (bind _greenroomplayerdisplay (Variables|Default|GetGreenRoomPlayerDisplay))
  (Utilities|IsValid _greenroomplayerdisplay
    (:Is%20Valid
      (return (Variables|/PilgrimQuickplay/Actors/GreenRoomPlayerDisplay/BP_GreenRoomPlayerDisplay.BP_GreenRoomPlayerDisplay_C|GetDisplayBoardWidget _greenroomplayerdisplay)))
    (:Is%20Not%20Valid
      (Fortnite|FortShippingLog 0 "ERROR: 'Get Display Board Widget' called in BP_PilgrimQuickplay_PlayerManager_PlayspaceComponent, but GreenRoomPlayerDisplay is not valid."))))

(fn GetTimerImage
  (bind _wbpdisplayboard (CallFunction|GetDisplayBoardWidget))
  (bind _stagetimer (Variables|/PilgrimQuickplay/Widgets/WBP_GreenRoomDisplayBoard.WBP_GreenRoomDisplayBoard_C|GetStageTimer _wbpdisplayboard))
  (bind _timerimage (Variables|/PilgrimQuickplay/Widgets/WBP_PilgrimQuickplay_Greenroom_StageTimer.WBP_PilgrimQuickplay_Greenroom_StageTimer_C|GetTimerImage _stagetimer))
  (Utilities|IsValid _wbpdisplayboard
    (:Is%20Valid
      (Utilities|IsValid _stagetimer
        (:Is%20Valid
          (Utilities|IsValid _timerimage
            (:Is%20Valid
              (return _timerimage))
            (:Is%20Not%20Valid
              _)))
        (:Is%20Not%20Valid
          _)))
    (:Is%20Not%20Valid
      _))
  (return 0))

(fn HandleNewPlayerVote
  (params (NewVote :type Pilgrim%20Premade%20Setlist%20Object%20Reference))
  (bind (_value _returnvalue) (Utilities|Map|Find (Variables|Default|GetVotedForSetlist_Server) NewVote))
  ;[1 1 1] Setting the new vote
  (Utilities|IsValid NewVote
    (:Is%20Valid
      (Utilities|Map|Add (Variables|Default|GetVotedForSetlist_Server) NewVote (Math|Integer|SelectInt (+ _value 1) 1 _returnvalue)))
    (:Is%20Not%20Valid
      _))
  (return))

(fn HandlePreviousPlayerVote
  (params (PreviousVote :type Pilgrim%20Premade%20Setlist%20Object%20Reference))
  (bind _votedforsetlist_server (Variables|Default|GetVotedForSetlist_Server))
  (bind (_value _returnvalue) (Utilities|Map|Find _votedforsetlist_server PreviousVote))
  ;[1 1 1] Removing the old vote
  (Utilities|IsValid PreviousVote
    (:Is%20Valid
      (if _returnvalue
        (if (== _value 1)
          (Utilities|Map|Remove (Variables|Default|GetVotedForSetlist_Server) PreviousVote)
          (else
            (Utilities|Map|Add _votedforsetlist_server PreviousVote (- _value 1))))
        (return)
        (else
          _)))
    (:Is%20Not%20Valid
      _)))

(fn HandleReadyTimerExpiration
  (CallFunction|MarkReadyTimerHasExpiredThisMatch)
  (Default|ForceAllPlayersReady))

(fn IsLastOrSingleSong
  ;[0.1356 0.4072 0.3763] If there is no next song we are on the last song of a set or are a single song
  (bind _asbp_pilgrim_quickplay_playspace (Utilities|Casting|CastToBP_PilgrimQuickplayPlayspace (Components|GetOwner (Variables|Default|GetEventRouter))))
  (return (not (Utilities|IsValid (Class|PilgrimCoreBandStateComponent|GetNextSong (Class|PilgrimQuickplayBandStateComponent|GetBandState _asbp_pilgrim_quickplay_playspace)) :!class /Script/Engine.KismetSystemLibrary))))

(fn IsPlayerOnLocalPlayerTeam
  (params (PlayerState :type Actor%20Object%20Reference))
  ;[0.2899 0.9927 1 0.577] Record this player's team.
  (LocalVariables|SetPlayerTeam (Team|GetActorTeam PlayerState))
  ;[0.2899 0.9927 1 0.577] Iterate through the player states until we find a locally controled player. Check their team against the passed in player, to see if they match.
  (bind _asbp_pilgrim_quickplay_playspace (Utilities|Casting|CastToBP_PilgrimQuickplayPlayspace (Components|GetOwner))
    (:then
      (for _array_element (Class|SparksMusicPlayspace|GetAllPlayerStates _asbp_pilgrim_quickplay_playspace)
        (Utilities|IsValid (Fortnite|GetFortPlayerControllerfromActor _array_element)
          (:Is%20Valid
            (if (Pawn|IsLocalPlayerController (Fortnite|GetFortPlayerControllerfromActor _array_element))
              (return (== (LocalVariables|GetPlayerTeam) (Team|GetActorTeam (Fortnite|GetFortPlayerControllerfromActor _array_element)))))))))
    (:CastFailed
      _))
  ;[0 0.87 0.8199 0.999] If we haven't already returned, return false.
  (return false))

(fn MarkReadyTimerHasExpiredThisMatch
  (bind _returnvalue (Components|GetOwner))
  (bind _returnvalue_2 (Actor|GetComponentbyClass _returnvalue "/Script/PilgrimCoreRuntime.PilgrimPlayerManager"))
  (Utilities|IsValid _returnvalue
    (:Is%20Valid
      (Utilities|IsValid _returnvalue_2
        (:Is%20Valid
          (Class|PilgrimPlayerManager|MarkReadyTimerHasExpiredThisMatch _returnvalue_2))))))

(fn OnRep_AllPlayersReady_Replicated
  (if (Variables|Default|GetAllPlayersReady_Replicated)
    (CallFunction|OnAllPlayersReady)))

(fn OnRep_MyPlayerStates
  (CallFunction|SetupMyPlayerStates))

(fn OnRep_PilgrimTimerState_Replicated
  (CallFunction|OnPilgrimTimerStateReplicated))

(fn OnRep_SongDisplayData
  (Default|UpdateVotesDisplaywithCurrentSetlists))

(fn PickWinner
  (Networking|CanExecuteCosmeticEvents
    (:True
      (Class|WBPGreenRoomVotesDisplay|PickWinner (Variables|/PilgrimQuickplay/Widgets/WBP_GreenRoomDisplayBoard.WBP_GreenRoomDisplayBoard_C|GetVotesDisplay (CallFunction|GetDisplayBoardWidget)) (Utilities|Struct|BreakSongDisplayDataStruct (Variables|Default|GetSongDisplayData))))))

(fn RecordPlayerVote
  (params (SetlistVote :type Pilgrim%20Premade%20Setlist%20Object%20Reference) (VoterPlayerState :type Player%20State%20Object%20Reference))
  (Utilities|IsValid SetlistVote
    (:Is%20Valid
      (Utilities|Map|Add (Variables|Default|GetPlayerVotes_Server) VoterPlayerState SetlistVote))
    (:Is%20Not%20Valid
      (Utilities|Map|Remove (Variables|Default|GetPlayerVotes_Server) VoterPlayerState)))
  (return))

(fn ResetReadyStateForAllPlayers
  (for _array_element (Variables|Default|GetAllPlayerStates)
    (Utilities|IsValid _array_element
      (:Is%20Valid
        (bind _returnvalue (Class|PilgrimQuickplayPlayerStateComponent|GetPilgrimQuickplayPlayerStateComponent _array_element))
        (Utilities|IsValid _returnvalue
          (:Is%20Valid
            (Class|PilgrimCorePlayerStateComponent|ServerSetIsReady _returnvalue false)
            (Class|PilgrimCorePlayerStateComponent|ServerSetIsReadytoReturn _returnvalue false))
          (:Is%20Not%20Valid
            (Development|PrintString "Cannot Reset Ready State.  PilgrimQuickplayPlayerState Component  is invalid"))))
      (:Is%20Not%20Valid
        (Development|PrintString "Cannot Reset Ready State.  PlayerState is invalid")))))

(fn SetPilgrimEventTimerState
  (params (PilgrimEventTimerState_GreenRoomTimerState_2_04FC39C74228C50F7AE0F7B72D379BE4 :type GreenRoomTimerState%20Enum) (PilgrimEventTimerState_FinalTimerLength_5_73A42F394BA6B986D4F14BA69DF7BFF0 :type Float%20%28double-precision%29))
  (bind (_pilgrimtimerstate_replicated_greenroomtimerstate_2_04fc39c74228c50f7ae0f7b72d379be4 _pilgrimtimerstate_replicated_finaltimerlength_5_73a42f394ba6b986d4f14ba69df7bff0) (Utilities|Struct|BreakPilgrimEventGreenRoomTimerStateChanged (Variables|GetPilgrimTimerState_Replicated_GreenRoomTimerState_2_04FC39C74228C50F7AE0F7B72D379BE4)))
  (if (not (and (Utilities|Enum|Equal(Enum) _pilgrimtimerstate_replicated_greenroomtimerstate_2_04fc39c74228c50f7ae0f7b72d379be4 PilgrimEventTimerState_GreenRoomTimerState_2_04FC39C74228C50F7AE0F7B72D379BE4) (== _pilgrimtimerstate_replicated_finaltimerlength_5_73a42f394ba6b986d4f14ba69df7bff0 PilgrimEventTimerState_FinalTimerLength_5_73A42F394BA6B986D4F14BA69DF7BFF0)))
    (Variables|SetPilgrimTimerState_Replicated_GreenRoomTimerState_2_04FC39C74228C50F7AE0F7B72D379BE4 (Utilities|Struct|MakePilgrimEventGreenRoomTimerStateChanged PilgrimEventTimerState_GreenRoomTimerState_2_04FC39C74228C50F7AE0F7B72D379BE4 PilgrimEventTimerState_FinalTimerLength_5_73A42F394BA6B986D4F14BA69DF7BFF0))))

(fn SetPulseTimerInteger
  (params (New%20Pulse%20Integer :type Integer))
  (bind _pilgrimtimerstate_replicated_greenroomtimerstate_2_04fc39c74228c50f7ae0f7b72d379be4 (Utilities|Struct|BreakPilgrimEventGreenRoomTimerStateChanged (Variables|GetPilgrimTimerState_Replicated_GreenRoomTimerState_2_04FC39C74228C50F7AE0F7B72D379BE4)))
  (bind _pulse_timer_count_timervalue_7_9833968f4a2479a71e2b2ba53665a4e7 (Utilities|Struct|BreakPilgrimEventGreenRoomTimerChanged (Variables|GetPulse%20Timer%20Count_TimerValue_7_9833968F4A2479A71E2B2BA53665A4E7)))
  (if (Variables|Default|GetSuppressGameplayMessages)
    _
    (elif (!= New%20Pulse%20Integer _pulse_timer_count_timervalue_7_9833968f4a2479a71e2b2ba53665a4e7)
      (Variables|SetPulse%20Timer%20Count_TimerValue_7_9833968F4A2479A71E2B2BA53665A4E7 (Utilities|Struct|MakePilgrimEventGreenRoomTimerChanged New%20Pulse%20Integer))
      (switch Utilities|FlowControl|Switch|SwitchonGreenRoomTimerState _pilgrimtimerstate_replicated_greenroomtimerstate_2_04fc39c74228c50f7ae0f7b72d379be4
        (:NewEnumerator1
          (Class|WBPPilgrimQuickplayGreenroomStageTimer|PulseNumber (Variables|/PilgrimQuickplay/Widgets/WBP_GreenRoomDisplayBoard.WBP_GreenRoomDisplayBoard_C|GetStageTimer (CallFunction|GetDisplayBoardWidget)) New%20Pulse%20Integer)
          (GameplayEventRouter|BroadcastEvent (Variables|Default|GetEventRouter) (Variables|Default|GetPulse%20Timer%20Count) 0))))))

(fn SetReturnToResultsInputEnabledForAll
  (params (EnableInputs :type Boolean))
  (for _array_element (Variables|Default|GetAllPlayerStates)
    (Utilities|IsValid _array_element
      (:Is%20Valid
        (Utilities|IsValid (PlayerState|GetPlayerController _array_element)
          (:Is%20Valid
            (Default|SetReturntoResultsInputEnabledForController (PlayerState|GetPlayerController _array_element) EnableInputs)))))))

(fn SetReturnToResultsInputEnabledforController
  (params (InController :type Controller%20Object%20Reference) (EnableInputs :type Boolean))
  (bind _incomingcontroller (LocalVariables|GetIncomingController))
  (bind _returnvalue (Pawn|GetControlledPawn _incomingcontroller))
  (bind _incomingcontroller_2 (LocalVariables|GetIncomingController))
  (bind _returnvalue_2 (Pawn|GetControlledPawn _incomingcontroller_2))
  (bind _returnvalue_3 (Ability|GetAbilitySystemComponent _returnvalue_2))
  (bind _returnvalue_4 (not EnableInputs))
  ;[1 1 1] Temporary Debugging log
  (Fortnite|FortShippingLog 0 (Utilities|String|Append "FORT-567677:  In bp_greenroomdpladisplay, Setting difficulty input enabled to" (Utilities|String|ToString(Boolean) EnableInputs)))
  (bind _output_get (LocalVariables|SetIncomingController InController))
  (Utilities|IsValid _output_get
    (:Is%20Valid
      (Utilities|FlowControl|Sequence
        (:then_0
          (if (Variables|Default|GetbBlockMovementWhenAllReady)
            (Input|SetIgnoreMoveInput (LocalVariables|GetIncomingController) _returnvalue_4)))
        (:then_1
          (Utilities|IsValid _returnvalue_2
            (:Is%20Valid
              (Utilities|IsValid _returnvalue_3
                (:Is%20Valid
                  (LocalVariables|SetLclAbilitySystemCmp _returnvalue_3)
                  (if _returnvalue_4
                    ;[1 1 1] Disable Input
                    (if (GameplayTags|HasMatchingGameplayTag _returnvalue_3 "(TagName=\"Pilgrim.GreenRoom.Input.ReturnToResults\")")
                      (GameplayEffects|RemoveActiveGameplayEffectbySourceEffect (LocalVariables|GetLclAbilitySystemCmp) (Variables|Default|GetEnable%20Return%20to%20Results%20Gameplay%20Effect%20Class) (LocalVariables|GetLclAbilitySystemCmp) -1))
                    (else
                      ;[1 1 1] Enable Input
                      (if (not (GameplayTags|HasMatchingGameplayTag (LocalVariables|GetLclAbilitySystemCmp) "(TagName=\"Pilgrim.GreenRoom.Input.ReturnToResults\")"))
                        (GameplayEffects|ApplyGameplayEffectToTarget :self (LocalVariables|GetLclAbilitySystemCmp) :GameplayEffectClass (Variables|Default|GetEnable%20Return%20to%20Results%20Gameplay%20Effect%20Class) :Target (LocalVariables|GetLclAbilitySystemCmp) :Level 0.0)))))))))
        (:then_2
          ;[1 1 1] Error Checking
          (Utilities|IsValid _returnvalue
            (:Is%20Valid
              (Utilities|IsValid (Ability|GetAbilitySystemComponent _returnvalue)
                (:Is%20Not%20Valid
                  (Development|PrintString "Incoming Controller's Pawn's AbilitySytemComp  is not valid." true true "(R=1.000000,G=0.000000,B=0.069961,A=1.000000)" 3e+01))))
            (:Is%20Not%20Valid
              (Development|PrintString "Incoming Controller's Pawn is not valid." true true "(R=1.000000,G=0.000000,B=0.004847,A=1.000000)" 3e+01))))))
    (:Is%20Not%20Valid
      (Development|PrintString "WARNING: Incoming Controller is not valid." true true "(R=1.000000,G=0.000000,B=0.219228,A=1.000000)" 3e+01))))

(fn SetSwapDifficultyInputEnabledForAll
  (params (EnableInputs :type Boolean))
  ;[1 1 1] Temporary Debugging Log
  (Fortnite|FortShippingLog 0 (Utilities|String|Append "FORT-567677:  In bp_greenroomdpladisplay, Setting difficulty input enabled to for all" (Utilities|String|ToString(Boolean) EnableInputs)))
  (for _array_element (Variables|Default|GetAllPlayerStates)
    (Utilities|IsValid _array_element
      (:Is%20Valid
        (Utilities|IsValid (PlayerState|GetPlayerController _array_element)
          (:Is%20Valid
            (Default|SetSwapDifficultyInputEnabledForController (PlayerState|GetPlayerController _array_element) EnableInputs)))))))

(fn SetSwapDifficultyInputEnabledforController
  (params (InController :type Controller%20Object%20Reference) (EnableInputs :type Boolean))
  (bind _incomingcontroller (LocalVariables|GetIncomingController))
  (bind _returnvalue (Pawn|GetControlledPawn _incomingcontroller))
  (bind _incomingcontroller_2 (LocalVariables|GetIncomingController))
  (bind _returnvalue_2 (Pawn|GetControlledPawn _incomingcontroller_2))
  (bind _returnvalue_3 (Ability|GetAbilitySystemComponent _returnvalue_2))
  ;[1 1 1] Temporary debugging Log
  (Fortnite|FortShippingLog 0 (Utilities|String|Append "FORT-567677:  In bp_greenroomdpladisplay, Setting difficulty input enabled to" (Utilities|String|ToString(Boolean) EnableInputs) (Utilities|GetDisplayName InController)))
  (bind _output_get (LocalVariables|SetIncomingController InController))
  (Utilities|IsValid _output_get
    (:Is%20Valid
      (Utilities|FlowControl|Sequence
        (:then_0
          (Utilities|IsValid _returnvalue_2
            (:Is%20Valid
              (Utilities|IsValid _returnvalue_3
                (:Is%20Valid
                  (LocalVariables|SetLclAbilitySystemCmp _returnvalue_3)
                  (if (not EnableInputs)
                    ;[1 1 1] Disable Input
                    (if (GameplayTags|HasMatchingGameplayTag _returnvalue_3 "(TagName=\"Pilgrim.GreenRoom.Input.SwapDifficulty\")")
                      (GameplayEffects|RemoveActiveGameplayEffectbySourceEffect (LocalVariables|GetLclAbilitySystemCmp) (Variables|Default|GetEnable%20Swap%20Difficulty%20Gameplay%20Effect%20Class) (LocalVariables|GetLclAbilitySystemCmp) -1))
                    (else
                      ;[1 1 1] Enable Input
                      (if (not (GameplayTags|HasMatchingGameplayTag (LocalVariables|GetLclAbilitySystemCmp) "(TagName=\"Pilgrim.GreenRoom.Input.SwapDifficulty\")"))
                        (GameplayEffects|ApplyGameplayEffectToTarget :self (LocalVariables|GetLclAbilitySystemCmp) :GameplayEffectClass (Variables|Default|GetEnable%20Swap%20Difficulty%20Gameplay%20Effect%20Class) :Target (LocalVariables|GetLclAbilitySystemCmp) :Level 0.0)))))))))
        (:then_1
          ;[1 1 1] Error Checking
          (Utilities|IsValid _returnvalue
            (:Is%20Valid
              (Utilities|IsValid (Ability|GetAbilitySystemComponent _returnvalue)
                (:Is%20Not%20Valid
                  (Development|PrintString "Incoming Controller's Pawn's AbilitySytemComp  is not valid." true true "(R=1.000000,G=0.000000,B=0.069961,A=1.000000)" 3e+01))))
            (:Is%20Not%20Valid
              (Development|PrintString "Incoming Controller's Pawn is not valid." true true "(R=1.000000,G=0.000000,B=0.004847,A=1.000000)" 3e+01))))))
    (:Is%20Not%20Valid
      (Development|PrintString "WARNING: Incoming Controller is not valid." true true "(R=1.000000,G=0.000000,B=0.219228,A=1.000000)" 3e+01))))

(fn SetToggleReadyUpInputEnabledForAll
  (params (EnableInputs :type Boolean))
  ;[1 1 1] Temporary debugging Log
  (Fortnite|FortShippingLog 0 (Utilities|String|Append "FORT-567677:  In bp_greenroomdpladisplay, Setting ready input enabled to" (Utilities|String|ToString(Boolean) EnableInputs)))
  (for _array_element (Variables|Default|GetAllPlayerStates)
    (Utilities|IsValid _array_element
      (:Is%20Valid
        (Utilities|IsValid (PlayerState|GetPlayerController _array_element)
          (:Is%20Valid
            (Default|SetToggleReadyUpInputEnabledForController (PlayerState|GetPlayerController _array_element) EnableInputs)))))))

(fn SetToggleReadyUpInputEnabledforController
  (params (InController :type Controller%20Object%20Reference) (EnableInputs :type Boolean))
  (bind _incomingcontroller (LocalVariables|GetIncomingController))
  (bind _returnvalue (Pawn|GetControlledPawn _incomingcontroller))
  (bind _incomingcontroller_2 (LocalVariables|GetIncomingController))
  (bind _returnvalue_2 (Pawn|GetControlledPawn _incomingcontroller_2))
  (bind _returnvalue_3 (Ability|GetAbilitySystemComponent _returnvalue_2))
  (bind _returnvalue_4 (not EnableInputs))
  ;[1 1 1] Temporary debugging Log
  (Fortnite|FortShippingLog 0 (Utilities|String|Append "FORT-567677:  In bp_greenroomdpladisplay, Setting ready input enabled to" (Utilities|String|ToString(Boolean) EnableInputs) (Utilities|GetDisplayName InController)))
  (bind _output_get (LocalVariables|SetIncomingController InController))
  (Utilities|IsValid _output_get
    (:Is%20Valid
      (Utilities|FlowControl|Sequence
        (:then_0
          (if (Variables|Default|GetbBlockMovementWhenAllReady)
            (Input|SetIgnoreMoveInput (LocalVariables|GetIncomingController) _returnvalue_4)))
        (:then_1
          (Utilities|IsValid _returnvalue_2
            (:Is%20Valid
              (Utilities|IsValid _returnvalue_3
                (:Is%20Valid
                  (LocalVariables|SetLclAbilitySystemCmp _returnvalue_3)
                  (if _returnvalue_4
                    ;[1 1 1] Disable Input
                    (if (GameplayTags|HasMatchingGameplayTag _returnvalue_3 "(TagName=\"Pilgrim.GreenRoom.Input.ToggleReadyUp\")")
                      (GameplayEffects|RemoveActiveGameplayEffectbySourceEffect (LocalVariables|GetLclAbilitySystemCmp) (Variables|Default|GetEnable%20Ready%20Up%20Gameplay%20Effect%20Class) (LocalVariables|GetLclAbilitySystemCmp) -1))
                    (else
                      ;[1 1 1] Enable Input
                      (if (not (GameplayTags|HasMatchingGameplayTag (LocalVariables|GetLclAbilitySystemCmp) "(TagName=\"Pilgrim.GreenRoom.Input.ToggleReadyUp\")"))
                        (GameplayEffects|ApplyGameplayEffectToTarget :self (LocalVariables|GetLclAbilitySystemCmp) :GameplayEffectClass (Variables|Default|GetEnable%20Ready%20Up%20Gameplay%20Effect%20Class) :Target (LocalVariables|GetLclAbilitySystemCmp) :Level 0.0)))))))))
        (:then_2
          ;[1 1 1] Error Checking
          (Utilities|IsValid _returnvalue
            (:Is%20Valid
              (Utilities|IsValid (Ability|GetAbilitySystemComponent _returnvalue)
                (:Is%20Not%20Valid
                  (Development|PrintString "Incoming Controller's Pawn's AbilitySytemComp  is not valid." true true "(R=1.000000,G=0.000000,B=0.069961,A=1.000000)" 3e+01))))
            (:Is%20Not%20Valid
              (Development|PrintString "Incoming Controller's Pawn is not valid." true true "(R=1.000000,G=0.000000,B=0.004847,A=1.000000)" 3e+01))))))
    (:Is%20Not%20Valid
      (Development|PrintString "WARNING: Incoming Controller is not valid." true true "(R=1.000000,G=0.000000,B=0.219228,A=1.000000)" 3e+01))))

(fn TryUpdateFromPlayerState
  (bind _wbpdisplayboard (CallFunction|GetGreenRoomPlayerDisplayWidgets))
  (bind _teamindex_local (LocalVariables|GetTeamIndex_Local))
  (bind _output (Utilities|Array|Get(acopy) _wbpdisplayboard _teamindex_local))
  (Networking|CanExecuteCosmeticEvents
    (:True
      (for (_array_element _array_index) (Variables|Default|GetMyPlayerStates)
        ;[0.2344 0.06035 0.2224] Don't check state we know we don't have widgets for!
        (if (< _array_index (Variables|Default|GetNumPlayerDisplayWidgets))
          (bind _output_get (LocalVariables|SetTeamIndex_Local (+ (LocalVariables|GetTeamIndex_Local) 1)))
          (if (Utilities|Array|IsValidIndex _wbpdisplayboard _output_get)
            (Utilities|IsValid _array_element
              (:Is%20Valid
                (bind _output_get_2 (Variables|/PilgrimQuickplay/Widgets/WBP_GreenRoomPlayerDisplay.WBP_GreenRoomPlayerDisplay_C|SetMyPlayerState _array_element _output))
                (Class|WBPGreenRoomPlayerDisplay|SetPlayer _output _output_get_2))
              (:Is%20Not%20Valid
                (Class|WBPGreenRoomPlayerDisplay|SettoDefault _output))) ; If the player state is invalid, set to default
            (else
              (CallFunction|TryPlayerStateRepAgain))))))))

(fn UpdateIndicatorsVisibilityForOpposingTeams)

(fn UpdateVotesDisplayWithCurrentSetlists
  (bind _songdisplaydata (Variables|Default|GetSongDisplayData))
  (bind (_votewinner_12_7c6c203e48fb20e81513929552288680 _setliststoshow_21_c2d8ba5f4ac1ba09c749a48a73fcb4df _playervoteskeys_20_810857544f5bdfd89209dfa8f065d6c0 _playervotessetlists_19_2c52d4ec45ae2cad3fa048b78bdf4123) (Utilities|Struct|BreakSongDisplayDataStruct _songdisplaydata))
  (Networking|CanExecuteCosmeticEvents
    (:True
      (for (_array_element _array_index) _playervoteskeys_20_810857544f5bdfd89209dfa8f065d6c0
        (Utilities|Map|Add (LocalVariables|GetRebuiltPlayerVotes) _array_element (Utilities|Array|Get(acopy) _playervotessetlists_19_2c52d4ec45ae2cad3fa048b78bdf4123 _array_index)))
      (Class|WBPGreenRoomVotesDisplay|RefreshDisplay (Variables|/PilgrimQuickplay/Widgets/WBP_GreenRoomDisplayBoard.WBP_GreenRoomDisplayBoard_C|GetVotesDisplay (CallFunction|GetDisplayBoardWidget)) _setliststoshow_21_c2d8ba5f4ac1ba09c749a48a73fcb4df (LocalVariables|GetRebuiltPlayerVotes)))))

(fn VoteNothingForAllVoters
  ;[1 1 1] Get the list of everyone we know that voted for a setlist and then set their vote to nothing to clear out votes for a new round
  (Networking|SwitchHasAuthority
    (:Authority
      (bind _keys (Utilities|Map|Keys (Variables|Default|GetPlayerVotes_Server)))
      (for _array_element _keys
        (bind _returnvalue (Class|PilgrimQuickplayPlayerStateComponent|GetPilgrimQuickplayPlayerStateComponent _array_element))
        (Class|PilgrimQuickplayPlayerStateComponent|SetSetlistVote _returnvalue 0))
      (return))))
