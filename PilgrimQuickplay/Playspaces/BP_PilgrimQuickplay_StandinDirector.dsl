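; Quickplay stand-in director component for Pilgrim that has stand-in player pawns equip their current instrument.
; On BeginPlay it waits for the owner's Pilgrim event router, then passes it to SetupListeners.
; Both equip functions call EquipCurrentInstrument on the pawn's alt-instrument component when one exists.
; TryEquipCorrectInstrument reports whether that component was found, while EquipCorrectInstrument always returns false.
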(event EventBeginPlay
  (bind _eventrouter (Pilgrim|PilgrimWaitForEventRouter (Components|GetOwner))
    (:OnEventRouterAvailable
      (CallFunction|SetupListeners _eventrouter))))

(fn EquipCorrectInstrument
  (params (Standin :type Fort%20Player%20Pawn%20Athena%20Object%20Reference))
  (bind _returnvalue (Actor|GetComponentbyClass Standin "/PilgrimCore/Player/Components/BP_PilgrimAltInstrumentPawnComponent.BP_PilgrimAltInstrumentPawnComponent_C"))
  (Utilities|IsValid _returnvalue
    (:Is%20Valid
      (Class|BPPilgrimAltInstrumentPawnComponent|EquipCurrentInstrument _returnvalue)
      (return false))
    (:Is%20Not%20Valid
      (return false))))

(fn TryEquipCorrectInstrument
  (params (Standin :type Fort%20Player%20Pawn%20Athena%20Object%20Reference))
  (bind _returnvalue (Actor|GetComponentbyClass Standin "/PilgrimCore/Player/Components/BP_PilgrimAltInstrumentPawnComponent.BP_PilgrimAltInstrumentPawnComponent_C"))
  (Utilities|IsValid _returnvalue
    (:Is%20Valid
      (Class|BPPilgrimAltInstrumentPawnComponent|EquipCurrentInstrument _returnvalue)
      (return true))
    (:Is%20Not%20Valid
      (return false))))
